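# Idle smoke effect for the HomeRange female character body, emitting two sprite layers, plain smoke and flipbook-animated detail, from seven skeletal-mesh bone/socket groups.
# Owner velocity drives spawn rate, lifetime, downward acceleration and inherited movement, so smoke gets denser and shorter-lived at speed.
# User.Variant selects the smoke color and User.Off scales spawning to zero.
# Alpha adjusts for front end versus gameplay and time of day.
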NiagaraSystem /BRCosmetics/Characters/Player/Female/Medium/Bodies/F_MED_HomeRange/FX/NS_Character_HomeRange_Idle
dependencies {
 usedDataInterfaces [
  &NiagaraDataInterfaceCurve
  &NiagaraDataInterfaceVector2DCurve
  &NiagaraDataInterfaceSpriteRendererInfo
  &NiagaraDataInterfaceSkeletalMesh
 ]
 usedDynamicInputs [
  &Vector2DFromCurve
  &RandomRangeVector2D
  &SetBoolByFloatComparison
 ]
 usedModules [
  &self:SystemSpawnScript.SystemScriptSource.SystemScriptGraph.NiagaraNodeAssignment_0.SetVariables_#1
  &SystemState
  &self:SystemSpawnScript.SystemScriptSource.SystemScriptGraph.NiagaraNodeAssignment_1.SetVariables_#2
  &self:SystemSpawnScript.SystemScriptSource.SystemScriptGraph.NiagaraNodeAssignment_2.SetVariables_#3
 ]
}
emitters {
 Smoke {
  emitterUpdateScript [
   &EmitterState(Inactive Response=::NewEnumerator1, Life Cycle Mode=::NewEnumerator1, Loop Delay=1.9, Loop Duration=1, UseLoopDelay=false)
   &SpawnRate(
    SpawnRate=&Multiply_Float(
     A=&Lerp_Float(A=1, B=0, Alpha=$User.Off)
     B=&Multiply_Float(
      A=&RandomRangeFloat(Minimum=8, Maximum=11)
      B=&Multiply_Float(
       A=1
       B=&FloatFromCurve(
        FloatCurve=DataInterface() {propertyValues {Curve {keys [<at 0 0.65 RCIM_Cubic>,<at 1 1.2>]}}}
        CurveIndex=&RemapRange(
         Input Value To Remap=&VectorLength(Vector=$Engine.Owner.Velocity)
         Input Min=0
         Input Max=1000
         Output Min=0
         Output Max=1
        )
       )
      )
     )
    )
    Use Spawn Probability=false
   )
  ]
  particleSpawnScript [
   &FN_SetSpawnGroup(Number of Spawn Groups=7)
   &InitializeParticle(
    Color=&MergeHelper_ColorSwitch(
     0=<rgba 0.05 0.05 0.05 1>
     1=<rgba 0.3404 0.8348 1 0.48>
     2=<rgba 0.7022 0.304 1 0.6>
     3=<rgba 1 1 1 0.38>
     Index=&MakeIntFromFloat(Float=$User.Variant, Conversion Mode=::NewEnumerator1)
     Num=::NewEnumerator2
    )
    Lifetime=&Multiply_Float(
     A=&RandomRangeFloat(Minimum=0.7, Maximum=1)
     B=&FloatFromCurve(
      FloatCurve=DataInterface() {propertyValues {Curve {keys [<at 0 1.4 RCIM_Cubic>,<at 1 0.45 RCIM_Cubic>]}}}
      CurveIndex=&RemapRange(
       Input Value To Remap=&VectorLength(Vector=$Engine.Owner.Velocity)
       Input Min=0
       Input Max=1000
       Output Min=0
       Output Max=1
      )
     )
    )
    Sprite Rotation Mode=::NewEnumerator2
    Sprite Size Mode=::NewEnumerator1
    Uniform Sprite Size Max=100
    Uniform Sprite Size Min=75
    UsePositionOffset=false
    Write Material Random=true
   )
   &CameraOffset(Camera Offset Amount=9)
   &SkeletalMeshLocation(
    Bone / Socket Index=&SpawnGroup()
    Bone Sampling Mode=::NewEnumerator4
    MaskSpawnGroup=false
    Sampled Position Offset=&MergeHelper_VectorSwitch(
     0=<10 0 0>
     1=<0 0 0>
     2=<-10 0 0>
     3=<0 0 0>
     4=<-7 0 -4>
     5=<7 0 4>
     6=<0 -8 0>
     7=<0 -5 0>
     8=<7 -3 0>
     9=<0 0 0>
     Index=&SpawnGroup()
     Num=::NewEnumerator8
    )
    Skeletal Mesh=$System.Skeletal Mesh
   )
  ]
  particleUpdateScript [
   &ParticleState()
   &SpriteRotationRate(Rotation Rate=&RandomRangeFloat(Minimum=-100, Maximum=100))
   &AccelerationForce(Acceleration=&Multiply_VectorByFloat(
     Vector=&RandomRangeVector(Minimum=<0 0 -100>, Maximum=<0 0 -50>)
     Float=&Multiply_Float(
      A=1
      B=&FloatFromCurve(
       FloatCurve=DataInterface() {propertyValues {Curve {keys [<at 0 0.5>,<at 1 1>]}}}
       CurveIndex=&RemapRange(
        Input Value To Remap=&VectorLength(Vector=$Engine.Owner.Velocity)
        Input Min=0
        Input Max=1000
        Output Min=0
        Output Max=1
       )
      )
     )
    ))
   &ScaleSpriteSize(
    Uniform Curve Scale=0.4
    Uniform Curve Sprite Scale=DataInterface() {propertyValues {Curve {
       defaultValue=1
       keys [
        {arriveTangent=0.5806, interpMode=RCIM_Cubic, leaveTangent=0.5806, tangentMode=RCTM_User, time=0, value=0.2923}
        {arriveTangent=0.0224, interpMode=RCIM_Cubic, leaveTangent=0.0224, tangentMode=RCTM_User, time=1, value=2}
       ]
      }}}
   )
   &InheritSourceMovement(
    Applied Position Scale=&VectorFromFloat(Value=&ClampFloat(Float=&Multiply_Float(
       A=&FloatFromCurve(
        FloatCurve=DataInterface() {propertyValues {Curve {keys [<at 0.6 1>,<at 0.95 1.1767>,<at 0.982 1.6>]}}}
        CurveIndex=&RemapRange(
         Input Value To Remap=&VectorLength(Vector=$Engine.Owner.Velocity)
         Input Min=0
         Input Max=1200
         Output Min=0
         Output Max=1
        )
       )
       B=&FloatFromCurve(FloatCurve=DataInterface() {propertyValues {Curve {keys [<at 0 1 RCIM_Cubic>,<at 1 0.55>]}}})
      )))
    ApplyForce=false
    ApplyPosition=true
    ApplyPositionFalloff=false
   )
   &SolveForcesAndVelocity(Clamp Velocity=false, Limit Acceleration=false)
   &ScaleColor(
    Scale Alpha=&FloatFromCurve(
     FloatCurve=DataInterface() {propertyValues {Curve {keys [
         {interpMode=RCIM_Cubic, tangentMode=RCTM_User, time=0, value=0}
         {arriveTangent=-2.0189, interpMode=RCIM_Cubic, leaveTangent=-2.1015, tangentMode=RCTM_Break, time=0.1949, value=1.35}
         <at 1 0 RCIM_Cubic>
        ]}}}
     Scale Curve=&Multiply_Float(
      A=&FN_FrontEndFloatModifier(Gameplay=1.2, Front End=0.8)
      B=&FNTimeOfDayScalar(Night Scale=0.6, Morning Scale=0.9, Evening Scale=0.8)
     )
    )
    ScaleA=true
    ScaleRGB=true
   )
  ]
  rendererClasses [&NiagaraSpriteRendererProperties]
  renderers [{rendererClass=&NiagaraSpriteRendererProperties}]
 }
 SmokeDetailed {
  emitterUpdateScript [
   &EmitterState(Inactive Response=::NewEnumerator1, Life Cycle Mode=::NewEnumerator1, Loop Delay=0.4, Loop Duration=1, UseLoopDelay=false)
   &SpawnRate(
    SpawnRate=&Multiply_Float(
     A=&Lerp_Float(A=1, B=0, Alpha=$User.Off)
     B=&Multiply_Float(
      A=&RandomRangeFloat(Minimum=11, Maximum=14)
      B=&Multiply_Float(
       A=1
       B=&FloatFromCurve(
        FloatCurve=DataInterface() {propertyValues {Curve {keys [<at 0 0.5 RCIM_Cubic>,<at 1 1.2>]}}}
        CurveIndex=&RemapRange(
         Input Value To Remap=&VectorLength(Vector=$Engine.Owner.Velocity)
         Input Min=0
         Input Max=1000
         Output Min=0
         Output Max=1
        )
       )
      )
     )
    )
    Use Spawn Probability=false
   )
  ]
  particleSpawnScript [
   &FN_SetSpawnGroup(Number of Spawn Groups=7)
   &InitializeParticle(
    Color=&MergeHelper_ColorSwitch(
     0=<rgba 0.05 0.05 0.05 0.828>
     1=<rgba 0.3404 0.8348 1 0.16>
     2=<rgba 0.6551 0.2442 1 0.26>
     3=<rgba 1 1 1 0.12>
     Index=&MakeIntFromFloat(Float=$User.Variant, Conversion Mode=::NewEnumerator1)
     Num=::NewEnumerator2
    )
    Lifetime=&Multiply_Float(
     A=&RandomRangeFloat(Minimum=1, Maximum=1.2)
     B=&FloatFromCurve(
      FloatCurve=DataInterface() {propertyValues {Curve {keys [<at 0 1.4 RCIM_Cubic>,<at 1 0.45 RCIM_Cubic>]}}}
      CurveIndex=&RemapRange(
       Input Value To Remap=&VectorLength(Vector=$Engine.Owner.Velocity)
       Input Min=0
       Input Max=1000
       Output Min=0
       Output Max=1
      )
     )
    )
    Sprite Rotation Mode=::NewEnumerator2
    Sprite Size Max=<100 100>
    Sprite Size Min=<80 80>
    Sprite Size Mode=::NewEnumerator2
    UsePositionOffset=false
    Write Material Random=true
   )
   &CameraOffset(Camera Offset Amount=9)
   &SkeletalMeshLocation(
    Bone / Socket Index=&SpawnGroup()
    Bone Sampling Mode=::NewEnumerator4
    MaskSpawnGroup=false
    Sampled Position Offset=&MergeHelper_VectorSwitch(
     0=<10 0 0>
     1=<0 0 0>
     2=<-10 0 0>
     3=<0 0 0>
     4=<-7 0 -4>
     5=<7 0 4>
     6=<0 -8 0>
     7=<0 -5 0>
     8=<7 -3 0>
     9=<0 0 0>
     Index=&SpawnGroup()
     Num=::NewEnumerator8
    )
    Skeletal Mesh=$System.Skeletal Mesh
   )
  ]
  particleUpdateScript [
   &ParticleState()
   &SpriteRotationRate(Rotation Rate=&RandomRangeFloat(Minimum=-50, Maximum=50))
   &AccelerationForce(Acceleration=&Multiply_VectorByFloat(
     Vector=&RandomRangeVector(Minimum=<0 0 -100>, Maximum=<0 0 -50>)
     Float=&Multiply_Float(
      A=1
      B=&FloatFromCurve(
       FloatCurve=DataInterface() {propertyValues {Curve {keys [<at 0 0.5>,<at 1 1>]}}}
       CurveIndex=&RemapRange(
        Input Value To Remap=&VectorLength(Vector=$Engine.Owner.Velocity)
        Input Min=0
        Input Max=1000
        Output Min=0
        Output Max=1
       )
      )
     )
    ))
   &ScaleSpriteSize(
    Scale Sprite Size Mode=::NewEnumerator0
    Uniform Scale Factor=&FloatFromCurve(
     FloatCurve=DataInterface() {propertyValues {Curve {
        defaultValue=1
        keys [
         {arriveTangent=1.4782, interpMode=RCIM_Cubic, leaveTangent=1.4782, tangentMode=RCTM_User, time=0, value=0.2923}
         {arriveTangent=0.0224, interpMode=RCIM_Cubic, leaveTangent=0.0224, tangentMode=RCTM_User, time=1, value=2}
        ]
       }}}
     Scale Curve=0.3
    )
   )
   &InheritSourceMovement(
    Applied Position Scale=&VectorFromFloat(Value=&ClampFloat(Float=&Multiply_Float(
       A=&FloatFromCurve(
        FloatCurve=DataInterface() {propertyValues {Curve {keys [<at 0.6 1>,<at 0.95 1.1767>,<at 0.982 1.6>]}}}
        CurveIndex=&RemapRange(
         Input Value To Remap=&VectorLength(Vector=$Engine.Owner.Velocity)
         Input Min=0
         Input Max=1200
         Output Min=0
         Output Max=1
        )
       )
       B=&FloatFromCurve(FloatCurve=DataInterface() {propertyValues {Curve {keys [<at 0 1 RCIM_Cubic>,<at 1 0.55>]}}})
      )))
    ApplyForce=false
    ApplyPosition=true
    ApplyPositionFalloff=false
   )
   &SubUVAnimation(
    End Frame Range Override=30
    "Sprite Renderer "=DataInterface() {propertyValues {SpriteRenderer=&self:SmokeDetailed_0.NiagaraSpriteRendererProperties_0}}
    Start Frame Range Override=&RandomRangeInt(Minimum=0, Maximum=5)
    UseEndFrame=true
    UseStartFrame=true
   )
   &ScaleColor(
    Scale Alpha=&Multiply_Float(
     A=&RandomRangeFloat(Minimum=0.8, Maximum=1)
     B=&FloatFromCurve(
      FloatCurve=DataInterface() {propertyValues {Curve {keys [
          {arriveTangent=-0.0994, interpMode=RCIM_Cubic, leaveTangent=-0.0994, tangentMode=RCTM_User, time=0, value=0}
          {interpMode=RCIM_Cubic, tangentMode=RCTM_User, time=0.2179, value=1}
          <at 1 0 RCIM_Cubic>
         ]}}}
      Scale Curve=&Multiply_Float(
       A=&FN_FrontEndFloatModifier(Gameplay=2, Front End=1.6)
       B=&FNTimeOfDayScalar(Night Scale=0.6, Morning Scale=0.9, Evening Scale=0.8)
      )
     )
    )
    ScaleA=true
    ScaleRGB=true
   )
   &SolveForcesAndVelocity(Clamp Velocity=false, Limit Acceleration=false)
  ]
  rendererClasses [&NiagaraSpriteRendererProperties]
  renderers [{rendererClass=&NiagaraSpriteRendererProperties}]
 }
}
summary {
 systemName=NS_Character_HomeRange_Idle
 emitters [
  {emitterName=Smoke, rendererClasses [&NiagaraSpriteRendererProperties]}
  {emitterName=SmokeDetailed, rendererClasses [&NiagaraSpriteRendererProperties]}
 ]
 userVariables [
  User.bIsFrontEnd {type=NiagaraBool}
  User.bIsFrontEndPreview {type=NiagaraBool}
  User.Off {type=NiagaraFloat, defaultValue=0, description=""}
  User.Variant {type=NiagaraFloat, defaultValue=0, description=""}
 ]
}
