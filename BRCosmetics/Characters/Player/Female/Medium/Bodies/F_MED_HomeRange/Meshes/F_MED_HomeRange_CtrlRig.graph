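# Control Rig for the HomeRange female medium player body, correcting outfit bones after the shared DeformRig function runs.
# Belt and gun bones use weighted parent constraints and thigh-raise attributes, with the gun also aimed at the calf.
# Collar and choker ring combine constraints with pose readers, the collar adding arm-raise attributes and a twist-free neck bone.
# Foot yaw drives mirrored jean flare offsets, while chain harmonics wave the flares during skydiving without a parachute.
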ControlRigBlueprint /BRCosmetics/Characters/Player/Female/Medium/Bodies/F_MED_HomeRange/Meshes/F_MED_HomeRange_CtrlRig
graphs [
 New Function_ContainedGraph {nodes [
   RigVMFunctionEntryNode()
   > RigVMFunction_Sequence = RigVMAggregateNode()
   RigVMFunctionReturnNode(ExecuteContext=@RigVMFunction_Sequence.C)
   Get_Animation_Attribute = RigVMDispatchNode(BoneName=None, Default=0, Name=deformRig_l_thigh_up_alpha) {template="DISPATCH_RigDispatch_GetAnimAttribute(in Name,in BoneName,in Default,out Value,out Found)"}
   &RigUnit_ModifyTransforms(
    ItemToModify="((Item=(Type=Bone,Name=\"deform_belt_01\"),Transform=(Rotation=(X=0,Y=0,Z=0,W=1),Translation=(X=3.4455,Y=0.8595,Z=0),Scale3D=(X=1,Y=1,Z=1))),(Item=(Type=Bone,Name=\"deform_belt_02\"),Transform=(Rotation=(X=-0.0689,Y=-0.0406,Z=0.0028,W=0.9968),Translation=(X=2. ...(truncated)"
    WeightMinimum=0
    Weight=@Get_Animation_Attribute.Value
    ExecutePin=@RigVMFunction_Sequence.B
   )
   Get_Animation_Attribute_1 = RigVMDispatchNode(BoneName=None, Default=0, Name=deformRig_r_thigh_up_alpha) {template="DISPATCH_RigDispatch_GetAnimAttribute(in Name,in BoneName,in Default,out Value,out Found)"}
   > &RigUnit_ModifyTransforms(
    ItemToModify="((Item=(Type=Bone,Name=\"deform_belt_06\"),Transform=(Rotation=(X=0,Y=0,Z=0,W=1),Translation=(X=2.0972,Y=0.2066,Z=0),Scale3D=(X=1,Y=1,Z=1))),(Item=(Type=Bone,Name=\"deform_belt_07\"),Transform=(Rotation=(X=0,Y=0,Z=0,W=1),Translation=(X=3.397,Y=0,Z=0),Scale3D=( ...(truncated)"
    WeightMinimum=0
    Weight=@Get_Animation_Attribute_1.Value
   )
   > &RigUnit_ModifyTransforms(
    ItemToModify=((Item=(Type=Bone,Name="None"),Transform=(Rotation=(X=0,Y=0,Z=0,W=1),Translation=(X=0,Y=0,Z=0),Scale3D=(X=1,Y=1,Z=1))))
    Weight=1
    WeightMinimum=0
   )
   &RigUnit_ParentConstraint(
    AdvancedSettings=(RotationOrderForFilter=XZY)
    Child=(Type=Bone,Name="deform_belt_01")
    Parents="((Item=(Type=Bone,Name=\"deform_groin_l\"),Weight=0.0141),(Item=(Type=Bone,Name=\"thigh_twist_01_l\"),Weight=0.0191),(Item=(Type=Bone,Name=\"spine_02\"),Weight=0.0319),(Item=(Type=Bone,Name=\"thigh_twist_01_r\"),Weight=0.0365),(Item=(Type=Bone,Name=\"deform_groin_r ...(truncated)"
    Weight=1
    ExecutePin=@RigVMFunction_Sequence.A
   )
   > &RigUnit_ParentConstraint(
    AdvancedSettings=(RotationOrderForFilter=XZY)
    Child=(Type=Bone,Name="deform_belt_02")
    Parents="((Item=(Type=Bone,Name=\"thigh_twist_01_r\"),Weight=0.0034),(Item=(Type=Bone,Name=\"spine_02\"),Weight=0.0246),(Item=(Type=Bone,Name=\"thigh_twist_01_l\"),Weight=0.0647),(Item=(Type=Bone,Name=\"deform_groin_l\"),Weight=0.1282),(Item=(Type=Bone,Name=\"pelvis\"),Weigh ...(truncated)"
    Weight=1
   )
   > &RigUnit_ParentConstraint(
    AdvancedSettings=(RotationOrderForFilter=XZY)
    Child=(Type=Bone,Name="deform_belt_03")
    Parents="((Item=(Type=Bone,Name=\"deform_glute_l\"),Weight=0.0072),(Item=(Type=Bone,Name=\"spine_02\"),Weight=0.0161),(Item=(Type=Bone,Name=\"deform_groin_l\"),Weight=0.0196),(Item=(Type=Bone,Name=\"thigh_twist_01_l\"),Weight=0.1248),(Item=(Type=Bone,Name=\"pelvis\"),Weight= ...(truncated)"
    Weight=1
   )
   > &RigUnit_ParentConstraint(
    AdvancedSettings=(RotationOrderForFilter=XZY)
    Child=(Type=Bone,Name="deform_belt_04")
    Parents="((Item=(Type=Bone,Name=\"thigh_l\"),Weight=0.0002),(Item=(Type=Bone,Name=\"spine_01\"),Weight=0.0055),(Item=(Type=Bone,Name=\"spine_02\"),Weight=0.0162),(Item=(Type=Bone,Name=\"thigh_twist_01_l\"),Weight=0.0873),(Item=(Type=Bone,Name=\"deform_glute_l\"),Weight=0.121 ...(truncated)"
    Weight=1
   )
   > &RigUnit_ParentConstraint(
    AdvancedSettings=(RotationOrderForFilter=XZY)
    Child=(Type=Bone,Name="deform_belt_05")
    Parents="((Item=(Type=Bone,Name=\"thigh_twist_01_l\"),Weight=0.0014),(Item=(Type=Bone,Name=\"spine_01\"),Weight=0.0018),(Item=(Type=Bone,Name=\"deform_glute_r\"),Weight=0.0019),(Item=(Type=Bone,Name=\"spine_02\"),Weight=0.0041),(Item=(Type=Bone,Name=\"deform_glute_l\"),Weigh ...(truncated)"
    Weight=1
   )
   > &RigUnit_ParentConstraint(
    AdvancedSettings=(RotationOrderForFilter=XZY)
    Child=(Type=Bone,Name="deform_belt_06")
    Parents="((Item=(Type=Bone,Name=\"thigh_r\"),Weight=0.0008),(Item=(Type=Bone,Name=\"spine_01\"),Weight=0.0041),(Item=(Type=Bone,Name=\"spine_02\"),Weight=0.0104),(Item=(Type=Bone,Name=\"thigh_twist_01_r\"),Weight=0.1027),(Item=(Type=Bone,Name=\"deform_glute_r\"),Weight=0.232 ...(truncated)"
    Weight=1
   )
   > &RigUnit_ParentConstraint(
    AdvancedSettings=(RotationOrderForFilter=XZY)
    Child=(Type=Bone,Name="deform_belt_07")
    Parents="((Item=(Type=Bone,Name=\"thigh_r\"),Weight=0.0012),(Item=(Type=Bone,Name=\"deform_groin_r\"),Weight=0.0167),(Item=(Type=Bone,Name=\"deform_glute_r\"),Weight=0.0791),(Item=(Type=Bone,Name=\"thigh_twist_01_r\"),Weight=0.3831),(Item=(Type=Bone,Name=\"pelvis\"),Weight=0 ...(truncated)"
    Weight=1
   )
   > &RigUnit_ParentConstraint(
    AdvancedSettings=(RotationOrderForFilter=XZY)
    Child=(Type=Bone,Name="deform_belt_08")
    Parents="((Item=(Type=Bone,Name=\"spine_02\"),Weight=0.0002),(Item=(Type=Bone,Name=\"deform_glute_r\"),Weight=0.0018),(Item=(Type=Bone,Name=\"thigh_r\"),Weight=0.0234),(Item=(Type=Bone,Name=\"deform_groin_r\"),Weight=0.2733),(Item=(Type=Bone,Name=\"pelvis\"),Weight=0.3378),( ...(truncated)"
    Weight=1
   )
   > &RigUnit_ParentConstraint(
    AdvancedSettings=(RotationOrderForFilter=XZY)
    Child=(Type=Bone,Name="deform_belt_09")
    Parents="((Item=(Type=Bone,Name=\"deform_groin_l\"),Weight=0.0002),(Item=(Type=Bone,Name=\"thigh_twist_01_l\"),Weight=0.0027),(Item=(Type=Bone,Name=\"thigh_r\"),Weight=0.0035),(Item=(Type=Bone,Name=\"spine_02\"),Weight=0.0056),(Item=(Type=Bone,Name=\"thigh_twist_01_r\"),Weig ...(truncated)"
    Weight=1
   )
  ]}
 New Function_ContainedGraph {nodes [
   RigVMFunctionEntryNode()
   > RigVMFunction_Sequence = RigVMAggregateNode()
   &RigUnit_ParentConstraint(
    AdvancedSettings=(RotationOrderForFilter=XZY)
    Child=(Type=Bone,Name="dyn_collar_l")
    Filter=(TranslationFilter=(bX=True,bY=True,bZ=True),RotationFilter=(bX=false,bY=false,bZ=false),ScaleFilter=(bX=True,bY=True,bZ=True))
    Parents="((Item=(Type=Bone,Name=\"spine_03\"),Weight=0.0001),(Item=(Type=Bone,Name=\"deform_upperarm_fr_l\"),Weight=0.0029),(Item=(Type=Bone,Name=\"upperarm_twist_01_l\"),Weight=0.0049),(Item=(Type=Bone,Name=\"deform_neck_01_notwist\"),Weight=0.0073),(Item=(Type=Bone,Name= ...(truncated)"
    Weight=1
    ExecutePin=@RigVMFunction_Sequence.B
   )
   > &RigUnit_ParentConstraint(
    AdvancedSettings=(RotationOrderForFilter=XZY)
    Child=(Type=Bone,Name="dyn_collar_r")
    Filter=(TranslationFilter=(bX=True,bY=True,bZ=True),RotationFilter=(bX=false,bY=false,bZ=false),ScaleFilter=(bX=True,bY=True,bZ=True))
    Parents="((Item=(Type=Bone,Name=\"deform_pec_r\"),Weight=0.0003),(Item=(Type=Bone,Name=\"deform_upperarm_fr_r\"),Weight=0.0029),(Item=(Type=Bone,Name=\"upperarm_twist_01_r\"),Weight=0.0049),(Item=(Type=Bone,Name=\"deform_neck_01_notwist\"),Weight=0.0229),(Item=(Type=Bone,N ...(truncated)"
    Weight=1
   )
   RigVMFunctionReturnNode(ExecuteContext=@RigVMFunction_Sequence.D)
   RigUnit_GetRelativeTransformForItem = &RigUnit_GetRelativeTransformForItem(Child=(Type=Bone,Name="neck_01"), Parent=(Type=Bone,Name="spine_05"), bChildInitial=False, bParentInitial=False)
   RigVMFunction_MathQuaternionSwingTwist = &RigVMFunction_MathQuaternionSwingTwist(TwistAxis=(X=1,Y=0,Z=0), Input=@RigUnit_GetRelativeTransformForItem.RelativeTransform.Rotation)
   RigUnit_SetRelativeTransformForItem = /Script/ControlRig.RigUnit_SetRelativeTransformForItem(
    Child=(Type=Bone,Name="deform_neck_01_notwist")
    Parent=(Type=Bone,Name="spine_05")
    Value=(Rotation=(X=0,Y=0,Z=0,W=1),Translation=(X=0,Y=0,Z=0),Scale3D=(X=1,Y=1,Z=1))
    Weight=1
    bParentInitial=False
    bPropagateToChildren=True
    Value.Scale3D=@RigUnit_GetRelativeTransformForItem.RelativeTransform.Scale3D
    Value.Translation=@RigUnit_GetRelativeTransformForItem.RelativeTransform.Translation
    Value.Rotation=@RigVMFunction_MathQuaternionSwingTwist.Swing
    ExecutePin=@RigVMFunction_Sequence.A
   ) {
    is=struct RigVMUnitNode
    template="Set Relative Transform::Execute(in Child,in Parent,in bParentInitial,in Value,in Weight,in bPropagateToChildren)"
   }
   n1 = &RigUnit_SphericalPoseReader(
    ActiveRegionScaleFactors=(PositiveWidth=1,NegativeWidth=2.4,PositiveHeight=2.4,NegativeHeight=2.4)
    ActiveRegionSize=0.1
    Debug=(bDrawDebug=false,bDraw2D=False,bDrawLocalAxes=False,DebugScale=25,DebugSegments=20,DebugThickness=0.25)
    DriverAxis=(X=1,Y=0,Z=0)
    DriverItem=(Type=Bone,Name="clavicle_l")
    FalloffRegionScaleFactors=(PositiveWidth=1,NegativeWidth=2,PositiveHeight=1,NegativeHeight=1)
    FalloffSize=0.24
    FlipHeightScaling=False
    FlipWidthScaling=False
    OptionalParentItem=(Type=Bone,Name="None")
    RotationOffset=(X=0,Y=-45,Z=0)
    ExecutePin=@RigVMFunction_Sequence.C
   )
   > n2 = &RigUnit_SphericalPoseReader(
    ActiveRegionScaleFactors=(PositiveWidth=1,NegativeWidth=2.4,PositiveHeight=2.4,NegativeHeight=2.4)
    ActiveRegionSize=0.1
    Debug=(bDrawDebug=false,bDraw2D=False,bDrawLocalAxes=False,DebugScale=25,DebugSegments=20,DebugThickness=0.25)
    DriverAxis=(X=-1,Y=0,Z=0)
    DriverItem=(Type=Bone,Name="clavicle_r")
    FalloffRegionScaleFactors=(PositiveWidth=1,NegativeWidth=1,PositiveHeight=1,NegativeHeight=1)
    FalloffSize=0.24
    FlipHeightScaling=False
    FlipWidthScaling=False
    OptionalParentItem=(Type=Bone,Name="None")
    RotationOffset=(X=0,Y=135,Z=0)
   )
   > Modify Transforms with Mirror = RigVMFunctionReferenceNode(
    Base Modify Data=((Item=(Type=Bone,Name="dyn_collar_l"),Transform=(Rotation=(X=0.0789,Y=0.0941,Z=0.0789,W=0.9893),Translation=(X=0,Y=0,Z=0),Scale3D=(X=1,Y=1,Z=1))))
    Mirror Modify Data=((Item=(Type=Bone,Name="dyn_collar_r"),Transform=(Rotation=(X=0,Y=0,Z=0,W=1),Translation=(X=0,Y=0,Z=0),Scale3D=(X=1,Y=1,Z=1))))
    Weight=@n1.OutputParam
    "Mirror Weight "=@n2.OutputParam
   )
   > n3 = &RigUnit_SphericalPoseReader(
    ActiveRegionScaleFactors=(PositiveWidth=2.7,NegativeWidth=1,PositiveHeight=2.7,NegativeHeight=1)
    ActiveRegionSize=0.1
    Debug=(bDrawDebug=false,bDraw2D=False,bDrawLocalAxes=False,DebugScale=25,DebugSegments=20,DebugThickness=0.25)
    DriverAxis=(X=-1,Y=0,Z=0)
    DriverItem=(Type=Bone,Name="upperarm_r")
    FalloffRegionScaleFactors=(PositiveWidth=1,NegativeWidth=1,PositiveHeight=1,NegativeHeight=1)
    FalloffSize=0.27
    FlipHeightScaling=False
    FlipWidthScaling=False
    OptionalParentItem=(Type=Bone,Name="None")
    RotationOffset=(X=-120,Y=0,Z=40)
   )
   > &RigUnit_ModifyTransforms(
    ItemToModify=((Item=(Type=Bone,Name="dyn_collar_r"),Transform=(Rotation=(X=-0.0306,Y=-0.3081,Z=0.012,W=0.9508),Translation=(X=0.3544,Y=-0.1825,Z=-0.9724),Scale3D=(X=1,Y=1,Z=1))))
    WeightMinimum=0
    Weight=@n3.OutputParam
   )
   Get_Animation_Attribute = RigVMDispatchNode(BoneName=None, Default=0, Name=deformRig_l_upperarm_up_alpha) {template="DISPATCH_RigDispatch_GetAnimAttribute(in Name,in BoneName,in Default,out Value,out Found)"}
   Get_Animation_Attribute_1 = RigVMDispatchNode(BoneName=None, Default=0, Name=deformRig_r_upperarm_up_alpha) {template="DISPATCH_RigDispatch_GetAnimAttribute(in Name,in BoneName,in Default,out Value,out Found)"}
   > Modify Transforms with Mirror_1 = RigVMFunctionReferenceNode(
    Base Modify Data=((Item=(Type=Bone,Name="dyn_collar_l"),Transform=(Rotation=(X=0,Y=0.1305,Z=0,W=0.9914),Translation=(X=0,Y=0,Z=-3),Scale3D=(X=1,Y=1,Z=1))))
    Mirror Modify Data=((Item=(Type=Bone,Name="dyn_collar_r"),Transform=(Rotation=(X=0,Y=0,Z=0,W=1),Translation=(X=0,Y=0,Z=0),Scale3D=(X=1,Y=1,Z=1))))
    Weight=@Get_Animation_Attribute.Value
    "Mirror Weight "=@Get_Animation_Attribute_1.Value
   )
   > &RigUnit_ModifyTransforms(
    ItemToModify=((Item=(Type=Bone,Name="deform_neck_01_notwist"),Transform=(Rotation=(X=0,Y=0,Z=0,W=1),Translation=(X=0,Y=0,Z=0),Scale3D=(X=1,Y=1,Z=1.1))))
    Weight=1
    WeightMinimum=0
   )
  ]}
 New Function_ContainedGraph {nodes [
   RigVMFunctionEntryNode()
   > &RigUnit_ParentConstraint(
    AdvancedSettings=(RotationOrderForFilter=XZY)
    Child=(Type=Bone,Name="deform_gun")
    Parents="((Item=(Type=Bone,Name=\"deform_glute_r\"),Weight=0.0018),(Item=(Type=Bone,Name=\"pelvis\"),Weight=0.1414),(Item=(Type=Bone,Name=\"thigh_r\"),Weight=0.1469),(Item=(Type=Bone,Name=\"deform_groin_r\"),Weight=0.166),(Item=(Type=Bone,Name=\"thigh_twist_01_r\"),Weight=0. ...(truncated)"
    Weight=1
   )
   > &RigUnit_AimConstraintLocalSpaceOffset(
    Child=(Type=Bone,Name="deform_gun")
    Filter=(bX=True,bY=True,bZ=True)
    Parents=((Item=(Type=Bone,Name="calf_r"),Weight=1))
    Weight=1
    WorldUp=(Target=(X=0,Y=0,Z=1),Kind=Location,Space=(Type=Bone,Name="calf_r"))
   )
   Get_Animation_Attribute = RigVMDispatchNode(BoneName=None, Default=0, Name=deformRig_r_thigh_up_alpha) {template="DISPATCH_RigDispatch_GetAnimAttribute(in Name,in BoneName,in Default,out Value,out Found)"}
   > &RigUnit_ModifyTransforms(
    ItemToModify=((Item=(Type=Bone,Name="deform_gun"),Transform=(Rotation=(X=0,Y=0,Z=0,W=1),Translation=(X=0,Y=0,Z=0),Scale3D=(X=1,Y=1,Z=1))))
    WeightMinimum=0
    Weight=@Get_Animation_Attribute.Value
   )
   > RigVMFunctionReturnNode()
  ]}
 New Function_ContainedGraph {nodes [
   RigVMFunctionEntryNode()
   > RigVMFunction_Sequence = RigVMAggregateNode()
   Get Delta Ref = RigVMFunctionReferenceNode(Item=(Type=Bone,Name="foot_l"), ExecuteContext=@RigVMFunction_Sequence.A)
   > Get Delta Ref_1 = RigVMFunctionReferenceNode(Item=(Type=Bone,Name="foot_r"))
   Remap = &RigVMFunction_MathDoubleRemap(SourceMaximum=0, SourceMinimum=-65, TargetMaximum=0, TargetMinimum=1, Value=@"Get Delta Ref".Result.Rotation.Yaw)
   > Modify Transforms with Mirror = RigVMFunctionReferenceNode(
    Base Modify Data="((Item=(Type=Bone,Name=\"dyn_jean_flares_front_l_base\"),Transform=(Rotation=(X=0,Y=-0.7071,Z=0,W=0.7071),Translation=(X=3,Y=0,Z=0),Scale3D=(X=1,Y=1,Z=1))),(Item=(Type=Bone,Name=\"dyn_jean_flares_inner_l_base\"),Transform=(Rotation=(X=0,Y=0,Z=-0.3827,W=0.9239) ...(truncated)"
    Mirror Modify Data="((Item=(Type=Bone,Name=\"dyn_jean_flares_front_r_base\"),Transform=(Rotation=(X=0,Y=0,Z=0,W=1),Translation=(X=0,Y=0,Z=0),Scale3D=(X=1,Y=1,Z=1))),(Item=(Type=Bone,Name=\"dyn_jean_flares_inner_r_base\"),Transform=(Rotation=(X=0,Y=0,Z=0,W=1),Translation=(X=0,Y=0, ...(truncated)"
    Weight=@Remap.Result
    "Mirror Weight "=&RigVMFunction_MathDoubleRemap(SourceMaximum=0, SourceMinimum=-65, TargetMaximum=0, TargetMinimum=1, Value=@"Get Delta Ref_1".Result.Rotation.Yaw).Result
   )
   RigVMFunctionReturnNode(ExecuteContext=@RigVMFunction_Sequence.C)
  ]}
 New Function_ContainedGraph {nodes [
   n1 = RigVMFunctionEntryNode()
   RigVMFunction_MathBoolNot = &RigVMFunction_MathBoolNot(Value=RigVMVariableNode() {variable=Parachute}.Value)
   RigVMFunction_MathBoolAnd = &RigVMFunction_MathBoolAnd(B=@RigVMFunction_MathBoolNot.Result, A=RigVMVariableNode() {variable=Skydiving}.Value)
   RigVMFunction_ControlFlowBranch = &RigVMFunction_ControlFlowBranch(ExecuteContext=@n1.ExecuteContext, Condition=@RigVMFunction_MathBoolAnd.Result)
   RigVMFunctionReturnNode(ExecuteContext=@RigVMFunction_ControlFlowBranch.Completed)
   n2 = RigVMDispatchNode(False=(X=50,Y=10,Z=0), True=(X=90,Y=50,Z=0), Condition=RigVMVariableNode() {variable="Skydive Diving  Down"}.Value) {template="DISPATCH_RigVMDispatch_If(in Condition,in True,in False,out Result)"}
   /Script/ControlRig.RigUnit_ChainHarmonicsPerItem(
    ChainRoot=(Type=Bone,Name="dyn_jean_flares_inner_l_base")
    DrawWorldOffset=(Rotation=(X=0,Y=0,Z=0,W=1),Translation=(X=0,Y=0,Z=0),Scale3D=(X=1,Y=1,Z=1))
    Pendulum=(bEnabled=true,PendulumStiffness=2,PendulumGravity=(X=0,Y=0,Z=0),PendulumBlend=0.75,PendulumDrag=0.98,PendulumMinimum=0,PendulumMaximum=0.1,PendulumEase=Linear,UnwindAxis=(X=0,Y=1,Z=0),UnwindMinimum=0.2,UnwindMaximum=0.05)
    Reach=(bEnabled=true,ReachTarget=(X=0,Y=0,Z=0),ReachAxis=(X=1,Y=0,Z=0),ReachMinimum=0,ReachMaximum=0,ReachEase=Linear)
    Wave=(bEnabled=True,WaveFrequency=(X=3,Y=2,Z=0),WaveAmplitude=(X=0.05,Y=0.15,Z=0),WaveOffset=(X=5,Y=0,Z=0),WaveNoise=(X=0,Y=0,Z=0),WaveMinimum=0,WaveMaximum=1,WaveEase=CircularEaseInOut)
    WaveCurve=(EditorCurveData=(Keys=((),(Time=0.5,Value=0.2),(Time=1,Value=1))))
    bDrawDebug=True
    Speed=@n2.Result
    ExecutePin=@RigVMFunction_ControlFlowBranch.True
   ) {is=struct RigVMUnitNode}
   > &RigUnit_ChainHarmonicsPerItem(
    ChainRoot=(Type=Bone,Name="dyn_jean_flares_outer_l_base")
    DrawWorldOffset=(Rotation=(X=0,Y=0,Z=0,W=1),Translation=(X=0,Y=0,Z=0),Scale3D=(X=1,Y=1,Z=1))
    Pendulum=(bEnabled=true,PendulumStiffness=2,PendulumGravity=(X=0,Y=0,Z=0),PendulumBlend=0.75,PendulumDrag=0.98,PendulumMinimum=0,PendulumMaximum=0.1,PendulumEase=Linear,UnwindAxis=(X=0,Y=1,Z=0),UnwindMinimum=0.2,UnwindMaximum=0.05)
    Reach=(bEnabled=true,ReachTarget=(X=0,Y=0,Z=0),ReachAxis=(X=1,Y=0,Z=0),ReachMinimum=0,ReachMaximum=0,ReachEase=Linear)
    Wave=(bEnabled=True,WaveFrequency=(X=1,Y=2,Z=0),WaveAmplitude=(X=0.1,Y=0.17,Z=0),WaveOffset=(X=0,Y=2,Z=0),WaveNoise=(X=0,Y=0,Z=0),WaveMinimum=0,WaveMaximum=1,WaveEase=CircularEaseInOut)
    WaveCurve=(EditorCurveData=(Keys=((),(Time=0.5,Value=0.2),(Time=1,Value=1))))
    bDrawDebug=True
    Speed=@n2.Result
   )
   > &RigUnit_ChainHarmonicsPerItem(
    ChainRoot=(Type=Bone,Name="dyn_jean_flares_inner_r_base")
    DrawWorldOffset=(Rotation=(X=0,Y=0,Z=0,W=1),Translation=(X=0,Y=0,Z=0),Scale3D=(X=1,Y=1,Z=1))
    Pendulum=(bEnabled=true,PendulumStiffness=2,PendulumGravity=(X=0,Y=0,Z=0),PendulumBlend=0.75,PendulumDrag=0.98,PendulumMinimum=0,PendulumMaximum=0.1,PendulumEase=Linear,UnwindAxis=(X=0,Y=1,Z=0),UnwindMinimum=0.2,UnwindMaximum=0.05)
    Reach=(bEnabled=true,ReachTarget=(X=0,Y=0,Z=0),ReachAxis=(X=1,Y=0,Z=0),ReachMinimum=0,ReachMaximum=0,ReachEase=Linear)
    Wave=(bEnabled=True,WaveFrequency=(X=3,Y=2,Z=0),WaveAmplitude=(X=0.05,Y=0.15,Z=0),WaveOffset=(X=5,Y=0,Z=0),WaveNoise=(X=0,Y=0,Z=0),WaveMinimum=0,WaveMaximum=1,WaveEase=CircularEaseInOut)
    WaveCurve=(EditorCurveData=(Keys=((),(Time=0.5,Value=0.2),(Time=1,Value=1))))
    bDrawDebug=True
    Speed=@n2.Result
   )
   > &RigUnit_ChainHarmonicsPerItem(
    ChainRoot=(Type=Bone,Name="dyn_jean_flares_outer_r_base")
    DrawWorldOffset=(Rotation=(X=0,Y=0,Z=0,W=1),Translation=(X=0,Y=0,Z=0),Scale3D=(X=1,Y=1,Z=1))
    Pendulum=(bEnabled=true,PendulumStiffness=2,PendulumGravity=(X=0,Y=0,Z=0),PendulumBlend=0.75,PendulumDrag=0.98,PendulumMinimum=0,PendulumMaximum=0.1,PendulumEase=Linear,UnwindAxis=(X=0,Y=1,Z=0),UnwindMinimum=0.2,UnwindMaximum=0.05)
    Reach=(bEnabled=true,ReachTarget=(X=0,Y=0,Z=0),ReachAxis=(X=1,Y=0,Z=0),ReachMinimum=0,ReachMaximum=0,ReachEase=Linear)
    Wave=(bEnabled=True,WaveFrequency=(X=1,Y=2,Z=0),WaveAmplitude=(X=0.1,Y=0.17,Z=0),WaveOffset=(X=0,Y=2,Z=0),WaveNoise=(X=0,Y=0,Z=0),WaveMinimum=0,WaveMaximum=1,WaveEase=CircularEaseInOut)
    WaveCurve=(EditorCurveData=(Keys=((),(Time=0.5,Value=0.2),(Time=1,Value=1))))
    bDrawDebug=True
    Speed=@n2.Result
   )
  ]}
 RigVMFunctionLibrary {nodes [
   deform belt = RigVMCollapseNode(ExecuteContext=())
   deform collar = RigVMCollapseNode(ExecuteContext=())
   deform flares = RigVMCollapseNode(ExecuteContext=())
   deform gun = RigVMCollapseNode(ExecuteContext=())
   jean skydiving = RigVMCollapseNode(ExecuteContext=())
  ]}
 RigVMFunction_Sequence_1_ContainedGraph {nodes [
   n1 = RigVMFunctionEntryNode()
   RigVMFunction_Sequence = &RigVMFunction_Sequence(ExecuteContext=@n1.ExecuteContext)
   n2 = &RigVMFunction_Sequence(ExecuteContext=@RigVMFunction_Sequence.B)
   RigVMFunctionReturnNode(A=@RigVMFunction_Sequence.A, B=@n2.A, C=@n2.B)
  ]}
 RigVMFunction_Sequence_1_ContainedGraph {nodes [
   n1 = RigVMFunctionEntryNode()
   RigVMFunction_Sequence = &RigVMFunction_Sequence(ExecuteContext=@n1.ExecuteContext)
   n2 = &RigVMFunction_Sequence(ExecuteContext=@RigVMFunction_Sequence.B)
   n3 = &RigVMFunction_Sequence(ExecuteContext=@n2.B)
   RigVMFunctionReturnNode(A=@RigVMFunction_Sequence.A, B=@n2.A, C=@n3.A, D=@n3.B)
  ]}
 RigVMFunction_Sequence_1_ContainedGraph {nodes [
   n1 = RigVMFunctionEntryNode()
   RigVMFunction_Sequence = &RigVMFunction_Sequence(ExecuteContext=@n1.ExecuteContext)
   n2 = &RigVMFunction_Sequence(ExecuteContext=@RigVMFunction_Sequence.B)
   RigVMFunctionReturnNode(A=@RigVMFunction_Sequence.A, B=@n2.A, C=@n2.B)
  ]}
 RigVMModel {nodes [
   RigUnit_BeginExecution = &RigUnit_BeginExecution()
   FN_DeformRig_FNC = RigVMFunctionReferenceNode(
    DeformRig_ Curve="(deformRig_l_upperarm_up_alpha_72_#1=True,deformRig_r_upperarm_up_alpha_73_#2=True,deformRig_l_upperarm_fr_alpha_74_#3=False,deformRig_r_upperarm_fr_alpha_75_CB51B80C ...(truncated)"
    DeformRig_Bool="(upperarm_1_#4=True,elbow_3_#5=True,wrist_7_#6=True,groin_8_#7=True,glute_17_#8=True,knee_9_40C408324CA261 ...(truncated)"
    DeformRig_Debug=(upperarm_debug_18_#4=False,wrist_debug_20_#6=False,lat_debug_26_#9=False,groinGlute_debug_27_#7=False)
    DeformRig_Value="(upperarm_up_18_#10=(Rotation=(X=0.0,Y=-0.0,Z=0.0,W=1.0),Translation=(X=0.0,Y=0.0,Z=8.0),Scale3D=(X=0.70,Y=1.0,Z=1.0)),upperarm_fwd_fr_19_#11=(Rotation=(X=0.0,Y=-0.0,Z=0.0,W=1.0),Translation=(X=0.0, ...(truncated)"
    ExecuteContext=@RigUnit_BeginExecution.ExecutePin
   )
   > deform belt = RigVMFunctionReferenceNode()
   > deform collar = RigVMFunctionReferenceNode()
   > deform gun = RigVMFunctionReferenceNode()
   > deform flares = RigVMFunctionReferenceNode()
   > jean skydiving = RigVMFunctionReferenceNode()
   > &RigUnit_ModifyTransforms(
    ItemToModify=((Item=(Type=Bone,Name="dyn_necklace_01"),Transform=(Rotation=(X=0,Y=0,Z=-0.0101,W=0.9999),Translation=(X=0,Y=-0.2069,Z=0),Scale3D=(X=1,Y=1,Z=1))))
    Weight=1
    WeightMinimum=0
   )
   > &RigUnit_ParentConstraint(
    AdvancedSettings=(RotationOrderForFilter=XZY)
    Child=(Type=Bone,Name="deform_choker_ring")
    Parents=((Item=(Type=Bone,Name="spine_05"),Weight=0.33),(Item=(Type=Bone,Name="neck_02"),Weight=0.33),(Item=(Type=Bone,Name="neck_01"),Weight=0.33))
    Weight=1
   )
   > n1 = &RigUnit_SphericalPoseReader(
    ActiveRegionScaleFactors=(PositiveWidth=6,NegativeWidth=1,PositiveHeight=6,NegativeHeight=6)
    ActiveRegionSize=0.1
    Debug=(bDrawDebug=true,bDraw2D=False,bDrawLocalAxes=False,DebugScale=25,DebugSegments=20,DebugThickness=0.25)
    DriverAxis=(X=1,Y=0,Z=0)
    DriverItem=(Type=Bone,Name="head")
    FalloffRegionScaleFactors=(PositiveWidth=1,NegativeWidth=1,PositiveHeight=1,NegativeHeight=1)
    FalloffSize=0.6
    FlipHeightScaling=False
    FlipWidthScaling=False
    OptionalParentItem=(Type=Bone,Name="spine_05")
    RotationOffset=(X=90,Y=0,Z=160)
   )
   > &RigUnit_ModifyTransforms(
    ItemToModify=((Item=(Type=Bone,Name="deform_choker_ring"),Transform=(Rotation=(X=0,Y=0,Z=0.1045,W=0.9945),Translation=(X=0.3,Y=0.7,Z=0),Scale3D=(X=1,Y=1,Z=1))))
    WeightMinimum=0
    Weight=@n1.OutputParam
   )
  ]}
]
hierarchy [
 root {loc=<0 0 0>, rot=<0 0 0 1>}
 pelvis {parent=root, loc=<0 -0.5301 98.0571>, rot=<0.0127 0.707 -0.0127 -0.707>}
 spine_01 {parent=pelvis, loc=<6.1208 -0.9328 0>, rot=<0 0 0.034 -0.9994>}
 spine_02 {parent=spine_01, loc=<6.5996 -0.2095 0>, rot=<0 0 0.0238 -0.9997>}
 spine_03 {parent=spine_02, loc=<6.4914 -0.1237 0>, rot=<0 0 0.0149 -0.9999>}
 spine_04 {parent=spine_03, loc=<6.9598 0.0794 0>, rot=<0 0 -0.0132 -0.9999>}
 spine_05 {parent=spine_04, loc=<6.9987 0 0>, rot=<0 0 -0.0015 -1>}
 neck_01 {parent=spine_05, loc=<11.3138 0.1529 0>, rot=<0 0 -0.2003 -0.9797>}
 neck_02 {parent=neck_01, loc=<4.5627 0 0>, rot=<0 0 0.0504 -0.9987>}
 head {parent=neck_02, loc=<5.3914 0 0>, rot=<0 0 0.115 -0.9934>}
 FACIAL_C_FacialRoot {parent=head, loc=<1.9478 0.1702 0>, rot=<0.4893 -0.5105 -0.4893 -0.5105>}
 FACIAL_C_UnderChin {parent=FACIAL_C_FacialRoot, loc=<0 8.6979 3.1231>, rot=<0.5006 -0.0001 0 -0.8657>}
 FACIAL_L_UnderChin {parent=FACIAL_C_FacialRoot, loc=<2.8532 7.5668 1.9742>, rot=<0.4033 -0.3549 0.0827 -0.8394>}
 FACIAL_R_UnderChin {parent=FACIAL_C_FacialRoot, loc=<-2.8497 7.5706 1.9768>, rot=<0.4033 0.3548 -0.0826 -0.8394>}
 FACIAL_C_Jaw {parent=FACIAL_C_FacialRoot, loc=<0 2.1215 1.058>, rot=<0.339 0 0 -0.9408>}
 FACIAL_C_Jawline {parent=FACIAL_C_Jaw, loc=<0.0001 2.4509 7.2971>, rot=<0.2619 -0.0001 0 -0.9651>}
 FACIAL_L_Jawline {parent=FACIAL_C_FacialRoot, loc=<4.0178 5.5495 3.4674>, rot=<0.2089 0.1578 -0.0546 -0.9636>}
 FACIAL_R_Jawline {parent=FACIAL_C_FacialRoot, loc=<-4.0164 5.5505 3.4692>, rot=<0.2088 -0.1579 0.0546 -0.9636>}
 deform_choker_ring {parent=neck_01, loc=<1.2436 4.945 -0.0007>, rot=<0 0 0 1>}
 clavicle_l {parent=spine_05, loc=<5.4831 -0.1053 -3.2946>, rot=<0.0465 -0.7289 -0.009 -0.6829>}
 upperarm_l {parent=clavicle_l, loc=<9.8658 -1.2136 -0.4105>, rot=<0.0164 -0.3245 -0.0269 -0.9454>}
 lowerarm_l {parent=upperarm_l, loc=<26.465 -0.8345 0.259>, rot=<0.0036 0.0161 -0.2298 -0.9731>}
 hand_l {parent=lowerarm_l, loc=<22.7355 -0.3203 -0.6105>, rot=<-0.6784 -0.0223 0.0735 -0.7307>}
 thumb_01_l {parent=hand_l, loc=<1.3118 -0.3191 -2.459>, rot=<0.6546 -0.4042 0.0664 -0.6354>}
 thumb_02_l {parent=thumb_01_l, loc=<5.9336 -0.0448 0.0401>, rot=<-0.0037 0.0213 0.298 -0.9543>}
 thumb_03_l {parent=thumb_02_l, loc=<2.2223 0.3682 -0.0907>, rot=<0.0118 -0.013 0.0066 -0.9998>}
 index_metacarpal_l {parent=hand_l, loc=<1.6535 0.0183 -2.96>, rot=<0.0158 -0.0124 -0.0173 -0.9996>}
 index_01_l {parent=index_metacarpal_l, loc=<8.3008 0.0357 -0.0524>, rot=<-0.0044 0.0485 0.0924 -0.9945>}
 index_02_l {parent=index_01_l, loc=<3.3587 -0.3004 0.0452>, rot=<-0.002 0.0027 0.3225 -0.9466>}
 index_03_l {parent=index_02_l, loc=<2.5578 -0.123 -0.0252>, rot=<-0.002 0.0067 0.2558 -0.9667>}
 middle_metacarpal_l {parent=hand_l, loc=<2.1856 -0.1725 -0.5984>, rot=<0.006 0.0168 -0.0329 -0.9993>}
 middle_01_l {parent=middle_metacarpal_l, loc=<7.3275 -0.2447 0.0129>, rot=<-0.0208 0.0042 0.0858 -0.9961>}
 middle_02_l {parent=middle_01_l, loc=<4.0123 0.1401 0.1175>, rot=<-0.0218 0.0258 0.3512 -0.9357>}
 middle_03_l {parent=middle_02_l, loc=<2.6077 0.0534 0.1373>, rot=<0.0082 0.0188 0.1708 -0.9851>}
 ring_metacarpal_l {parent=hand_l, loc=<2.2074 -0.3625 1.3638>, rot=<-0.0284 0.0722 -0.0458 -0.9959>}
 ring_01_l {parent=ring_metacarpal_l, loc=<7.0295 -0.6197 -0.1446>, rot=<-0.0489 -0.0252 0.1892 -0.9804>}
 ring_02_l {parent=ring_01_l, loc=<3.6174 0.4647 -0.1027>, rot=<0 0 0.2806 -0.9598>}
 ring_03_l {parent=ring_02_l, loc=<2.4152 0.114 0.0199>, rot=<0 0 0.1189 -0.9929>}
 pinky_metacarpal_l {parent=hand_l, loc=<1.7616 -0.9753 3.3814>, rot=<-0.0397 0.0814 -0.069 -0.9935>}
 pinky_01_l {parent=pinky_metacarpal_l, loc=<7.1709 -0.4831 -0.1094>, rot=<-0.0654 -0.0035 0.2242 -0.9723>}
 pinky_02_l {parent=pinky_01_l, loc=<2.8274 0.0893 -0.1043>, rot=<0.0037 0.0011 0.2854 -0.9584>}
 pinky_03_l {parent=pinky_02_l, loc=<2.2046 0.0307 -0.0078>, rot=<-0.0002 0.0098 0.1445 -0.9895>}
 deform_wrist_dn_base_l {parent=hand_l, loc=<0 -1.5326 0>, rot=<0 0 0 -1>}
 deform_wrist_dn_l {parent=deform_wrist_dn_base_l, loc=<0 0 0>, rot=<0 0 0 1>}
 deform_wrist_up_base_l {parent=hand_l, loc=<0 2.3009 0>, rot=<0 0 0 -1>}
 deform_wrist_up_l {parent=deform_wrist_up_base_l, loc=<0 0 0>, rot=<0 0 0 1>}
 lowerarm_twist_02_l {parent=lowerarm_l, loc=<10.6707 0 0>, rot=<0 0 0 1>}
 lowerarm_twist_01_l {parent=lowerarm_l, loc=<18.316 0 0>, rot=<0 0 0 1>}
 deform_elbow_in_base_l {parent=lowerarm_l, loc=<0.1834 3.6062 -0.2572>, rot=<0 0 0 -1>}
 deform_elbow_in_l {parent=deform_elbow_in_base_l, loc=<0 0 0>, rot=<0 0 0 1>}
 deform_elbow_out_base_l {parent=lowerarm_l, loc=<0 -4 0>, rot=<0 0 0 1>}
 deform_elbow_out_l {parent=deform_elbow_out_base_l, loc=<0 0 0>, rot=<0 0 0 1>}
 upperarm_twist_01_l {parent=upperarm_l, loc=<7.2135 -0.1134 0>, rot=<0 0 0 1>}
 deform_upperarm_bk_base_l {parent=upperarm_twist_01_l, loc=<-6.4053 -4.3746 0.314>, rot=<0 0 0 -1>}
 deform_upperarm_bk_l {parent=deform_upperarm_bk_base_l, loc=<0 0 0>, rot=<0 0 0 1>}
 deform_upperarm_fr_base_l {parent=upperarm_twist_01_l, loc=<-5.6312 5.092 0.3354>, rot=<0 0 0 -1>}
 deform_upperarm_fr_l {parent=deform_upperarm_fr_base_l, loc=<0 0 0>, rot=<0 0 0 1>}
 deform_upperarm_up_base_l {parent=upperarm_twist_01_l, loc=<-8.891 0.4825 3.9517>, rot=<0 0 0 -1>}
 deform_upperarm_up_l {parent=deform_upperarm_up_base_l, loc=<0 0 0>, rot=<0 0 0 1>}
 upperarm_twist_02_l {parent=upperarm_l, loc=<19.7998 -0.1134 0>, rot=<0 0 0 1>}
 clavicle_r {parent=spine_05, loc=<5.4831 -0.1053 3.2946>, rot=<0.7289 0.0465 -0.6829 0.009>}
 upperarm_r {parent=clavicle_r, loc=<-9.8658 1.2136 0.4105>, rot=<0.0164 -0.3245 -0.0269 -0.9454>}
 lowerarm_r {parent=upperarm_r, loc=<-26.465 0.8345 -0.259>, rot=<0.0036 0.0161 -0.2298 -0.9731>}
 hand_r {parent=lowerarm_r, loc=<-22.7355 0.3203 0.6106>, rot=<-0.6784 -0.0223 0.0735 -0.7307>}
 thumb_01_r {parent=hand_r, loc=<-1.3118 0.3191 2.459>, rot=<0.6546 -0.4042 0.0664 -0.6354>}
 thumb_02_r {parent=thumb_01_r, loc=<-5.9336 0.0448 -0.0401>, rot=<-0.0037 0.0213 0.298 -0.9543>}
 thumb_03_r {parent=thumb_02_r, loc=<-2.2223 -0.3682 0.0907>, rot=<0.0118 -0.013 0.0066 -0.9998>}
 index_metacarpal_r {parent=hand_r, loc=<-1.6535 -0.0183 2.96>, rot=<0.0158 -0.0124 -0.0173 -0.9996>}
 index_01_r {parent=index_metacarpal_r, loc=<-8.3008 -0.0357 0.0524>, rot=<-0.0044 0.0485 0.0924 -0.9945>}
 index_02_r {parent=index_01_r, loc=<-3.3587 0.3004 -0.0452>, rot=<-0.002 0.0027 0.3225 -0.9466>}
 index_03_r {parent=index_02_r, loc=<-2.5578 0.123 0.0252>, rot=<-0.002 0.0067 0.2558 -0.9667>}
 middle_metacarpal_r {parent=hand_r, loc=<-2.1856 0.1725 0.5984>, rot=<0.006 0.0168 -0.0329 -0.9993>}
 middle_01_r {parent=middle_metacarpal_r, loc=<-7.3275 0.2447 -0.0129>, rot=<-0.0208 0.0042 0.0858 -0.9961>}
 middle_02_r {parent=middle_01_r, loc=<-4.0123 -0.1401 -0.1175>, rot=<-0.0218 0.0258 0.3512 -0.9357>}
 middle_03_r {parent=middle_02_r, loc=<-2.6077 -0.0534 -0.1373>, rot=<0.0082 0.0188 0.1708 -0.9851>}
 ring_metacarpal_r {parent=hand_r, loc=<-2.2074 0.3625 -1.3638>, rot=<-0.0284 0.0722 -0.0458 -0.9959>}
 ring_01_r {parent=ring_metacarpal_r, loc=<-7.0295 0.6197 0.1446>, rot=<-0.0489 -0.0252 0.1892 -0.9804>}
 ring_02_r {parent=ring_01_r, loc=<-3.6174 -0.4647 0.1027>, rot=<0 0 0.2806 -0.9598>}
 ring_03_r {parent=ring_02_r, loc=<-2.4152 -0.114 -0.0199>, rot=<0 0 0.1189 -0.9929>}
 pinky_metacarpal_r {parent=hand_r, loc=<-1.7616 0.9753 -3.3814>, rot=<-0.0397 0.0814 -0.069 -0.9935>}
 pinky_01_r {parent=pinky_metacarpal_r, loc=<-7.1709 0.4831 0.1094>, rot=<-0.0654 -0.0035 0.2242 -0.9723>}
 pinky_02_r {parent=pinky_01_r, loc=<-2.8274 -0.0893 0.1043>, rot=<0.0037 0.0011 0.2854 -0.9584>}
 pinky_03_r {parent=pinky_02_r, loc=<-2.2046 -0.0307 0.0078>, rot=<-0.0002 0.0098 0.1445 -0.9895>}
 deform_wrist_dn_base_r {parent=hand_r, loc=<-0.0001 1.5328 0>, rot=<0 0 0 -1>}
 deform_wrist_dn_r {parent=deform_wrist_dn_base_r, loc=<0 0 0>, rot=<0 0 0 1>}
 deform_wrist_up_base_r {parent=hand_r, loc=<-0.0002 -2.3007 -0.0001>, rot=<0 0 0 -1>}
 deform_wrist_up_r {parent=deform_wrist_up_base_r, loc=<0 0 0>, rot=<0 0 0 1>}
 lowerarm_twist_02_r {parent=lowerarm_r, loc=<-10.6707 0 0>, rot=<0 0 0 -1>}
 lowerarm_twist_01_r {parent=lowerarm_r, loc=<-18.316 0 0>, rot=<0 0 0 -1>}
 deform_elbow_in_base_r {parent=lowerarm_r, loc=<-0.1834 -3.6062 0.2572>, rot=<0 0 0 -1>}
 deform_elbow_in_r {parent=deform_elbow_in_base_r, loc=<0 0 0>, rot=<0 0 0 1>}
 deform_elbow_out_base_r {parent=lowerarm_r, loc=<0 4 -0.0001>, rot=<0 0 0 -1>}
 deform_elbow_out_r {parent=deform_elbow_out_base_r, loc=<0 0 0>, rot=<0 0 0 1>}
 upperarm_twist_01_r {parent=upperarm_r, loc=<-7.2135 0.1134 0>, rot=<0 0 0 -1>}
 deform_upperarm_bk_base_r {parent=upperarm_twist_01_r, loc=<6.4053 4.3746 -0.314>, rot=<0 0 0 -1>}
 deform_upperarm_bk_r {parent=deform_upperarm_bk_base_r, loc=<0 0 0>, rot=<0 0 0 1>}
 deform_upperarm_fr_base_r {parent=upperarm_twist_01_r, loc=<5.6312 -5.092 -0.3354>, rot=<0 0 0 -1>}
 deform_upperarm_fr_r {parent=deform_upperarm_fr_base_r, loc=<0 0 0>, rot=<0 0 0 1>}
 deform_upperarm_up_base_r {parent=upperarm_twist_01_r, loc=<8.891 -0.4825 -3.9517>, rot=<0 0 0 -1>}
 deform_upperarm_up_r {parent=deform_upperarm_up_base_r, loc=<0 0 0>, rot=<0 0 0 1>}
 upperarm_twist_02_r {parent=upperarm_r, loc=<-19.7998 0.1134 0>, rot=<0 0 0 -1>}
 attach_backpack {parent=spine_05, loc=<-0.123 -13.473 0>, rot=<-0.0093 -0.707 0.0093 -0.707>}
 Backpack_BR {parent=attach_backpack, type=Null, loc=<0 3.8659 -3.2786>, rot=<0 0 0 1>}
 attach_cape {parent=spine_05, loc=<9.017 -11.714 0>, rot=<0.0616 -0.7044 -0.0616 -0.7044>}
 cape {parent=attach_cape, type=Null, loc=<0 4.2928 -0.6677>, rot=<0 0 0 1>}
 deform_neck_01_notwist {parent=spine_05, loc=<11.3138 0.1529 0>, rot=<0 0 -0.2003 -0.9797>}
 dyn_necklace_01 {parent=spine_05, loc=<8.4936 3.3981 -0.0547>, rot=<0 0 0.9438 0.3304>}
 dyn_necklace_02 {parent=dyn_necklace_01, loc=<2.9367 0 0>, rot=<0 0 0 1>}
 dyn_necklace_03 {parent=dyn_necklace_02, loc=<1.8676 0 0>, rot=<0 0 0 1>}
 dyn_necklace_04 {parent=dyn_necklace_03, loc=<1.9403 0 0>, rot=<0 0 0 1>}
 dyn_necklace_05 {parent=dyn_necklace_04, loc=<0.7978 -0.0725 0>, rot=<0 0 0 1>}
 dyn_collar_l {parent=spine_05, loc=<7.344 2.6775 -6.8864>, rot=<-0.0515 0.1389 -0.2437 -0.9585>}
 dyn_collar_r {parent=spine_05, loc=<7.344 2.6775 6.8864>, rot=<0.1389 0.0515 0.9585 -0.2437>}
 deform_pec_aim_base_l {parent=spine_04, loc=<3.2456 9.3309 -1.2237>, rot=<0.0603 -0.4033 0.3318 -0.8507>}
 deform_pec_aim_l {parent=deform_pec_aim_base_l, loc=<0 0 0>, rot=<0 0 0 1>}
 deform_pec_base_l {parent=deform_pec_aim_l, loc=<9.3665 0 0>, rot=<0 0 0 1>}
 deform_pec_l {parent=deform_pec_base_l, loc=<0 0 0>, rot=<0 0 0 1>}
 deform_pec_aim_base_r {parent=spine_04, loc=<3.2452 9.3309 1.2237>, rot=<-0.4033 -0.0603 0.8507 0.3318>}
 deform_pec_aim_r {parent=deform_pec_aim_base_r, loc=<0 0 0>, rot=<0 0 0 -1>}
 deform_pec_base_r {parent=deform_pec_aim_r, loc=<-9.3667 -0.0003 -0.0004>, rot=<0 0 0 1>}
 deform_pec_r {parent=deform_pec_base_r, loc=<0 0 0>, rot=<0 0 0 1>}
 thigh_l {parent=pelvis, loc=<-1.0855 -0.2109 -8.7053>, rot=<0.0009 0.0355 0.0155 -0.9993>}
 calf_l {parent=thigh_l, loc=<-42.5284 -0.0019 -0.7297>, rot=<0.0006 -0.0161 -0.0391 -0.9991>}
 foot_l {parent=calf_l, loc=<-46.7401 -2.5014 -0.592>, rot=<0.0427 -0.0238 0.0353 -0.9982>}
 ball_l {parent=foot_l, loc=<-3.5658 10.8066 -0.1544>, rot=<0.0001 -0.0001 -0.7063 -0.7079>}
 dyn_spur_l_base {parent=foot_l, loc=<0.383 -10.5392 0.2303>, rot=<0.0028 -0.7058 0.0102 -0.7084>}
 dyn_spur_l {parent=dyn_spur_l_base, loc=<0 0 0>, rot=<0 0 0 1>}
 calf_twist_02_l {parent=calf_l, loc=<-19.1546 -1.7943 -0.2638>, rot=<0.0237 0.0095 -0.0035 -0.9997>}
 calf_twist_01_l {parent=calf_l, loc=<-35.8483 -1.9197 -0.5794>, rot=<0.0237 0.0095 -0.0035 -0.9997>}
 dyn_jean_flares_l_base {parent=calf_twist_01_l, loc=<0 0 0>, rot=<0 0 0 1>}
 dyn_jean_flares_front_l_base {parent=dyn_jean_flares_l_base, loc=<1.7094 5.6751 -0.4887>, rot=<-0.6651 0.0777 0.1589 -0.7255>}
 dyn_jean_flares_front_l_01 {parent=dyn_jean_flares_front_l_base, loc=<0 0 0>, rot=<0 0 0 1>}
 dyn_jean_flares_front_l_02 {parent=dyn_jean_flares_front_l_01, loc=<-7.428 0 0>, rot=<0 0 0 1>}
 dyn_jean_flares_inner_l_base {parent=dyn_jean_flares_l_base, loc=<1.3609 -1.9284 4.6321>, rot=<0.0273 -0.0978 -0.0086 -0.9948>}
 dyn_jean_flares_inner_l_01 {parent=dyn_jean_flares_inner_l_base, loc=<0 0 0>, rot=<0 0 0 1>}
 dyn_jean_flares_inner_l_02 {parent=dyn_jean_flares_inner_l_01, loc=<-7.428 0 0>, rot=<0 0 0 1>}
 dyn_jean_flares_outer_l_base {parent=dyn_jean_flares_l_base, loc=<0.7493 -1.5974 -5.6516>, rot=<0.0509 0.059 0.0453 -0.9959>}
 dyn_jean_flares_outer_l_01 {parent=dyn_jean_flares_outer_l_base, loc=<0 0 0>, rot=<0 0 0 1>}
 dyn_jean_flares_outer_l_02 {parent=dyn_jean_flares_outer_l_01, loc=<-7.428 0 0>, rot=<0 0 0 1>}
 deform_knee_in_base_l {parent=calf_l, loc=<0.8031 -5.0231 0.2456>, rot=<0 0 0 -1>}
 deform_knee_in_l {parent=deform_knee_in_base_l, loc=<0 0 0>, rot=<0 0 0 1>}
 deform_knee_out_base_l {parent=calf_l, loc=<0 4 0>, rot=<0 0 0 -1>}
 deform_knee_out_l {parent=deform_knee_out_base_l, loc=<0 0 0>, rot=<0 0 0 1>}
 thigh_twist_01_l {parent=thigh_l, loc=<-27.952 0 0>, rot=<0.0013 0 0 -1>}
 deform_groin_base_l {parent=thigh_l, loc=<0.4801 5.6776 -0.2486>, rot=<0 0 0 -1>}
 deform_groin_l {parent=deform_groin_base_l, loc=<0 0 0>, rot=<0 0 0 1>}
 thigh_r {parent=pelvis, loc=<-1.0855 -0.2109 8.7053>, rot=<0.0355 -0.0009 0.9993 0.0155>}
 calf_r {parent=thigh_r, loc=<42.5284 0.0019 0.7297>, rot=<0.0006 -0.0161 -0.0391 -0.9991>}
 foot_r {parent=calf_r, loc=<46.7401 2.5014 0.592>, rot=<0.0427 -0.0238 0.0353 -0.9982>}
 ball_r {parent=foot_r, loc=<3.5658 -10.8066 0.1544>, rot=<0.0001 -0.0001 -0.7063 -0.7079>}
 dyn_spur_r_base {parent=foot_r, loc=<-0.383 10.5392 -0.2303>, rot=<0.0028 -0.7058 0.0102 -0.7084>}
 dyn_spur_r {parent=dyn_spur_r_base, loc=<0 0 0>, rot=<0 0 0 1>}
 calf_twist_02_r {parent=calf_r, loc=<19.1546 1.7943 0.2638>, rot=<0.0237 0.0095 -0.0035 -0.9997>}
 calf_twist_01_r {parent=calf_r, loc=<35.8483 1.9197 0.5794>, rot=<0.0237 0.0095 -0.0035 -0.9997>}
 dyn_jean_flares_r_base {parent=calf_twist_01_r, loc=<0 0 0>, rot=<0 0 0 1>}
 dyn_jean_flares_front_r_base {parent=dyn_jean_flares_r_base, loc=<-1.7094 -5.6751 0.4887>, rot=<-0.6651 0.0777 0.1589 -0.7255>}
 dyn_jean_flares_front_r_01 {parent=dyn_jean_flares_front_r_base, loc=<0 0 0>, rot=<0 0 0 1>}
 dyn_jean_flares_front_r_02 {parent=dyn_jean_flares_front_r_01, loc=<7.428 0 0>, rot=<0 0 0 1>}
 dyn_jean_flares_inner_r_base {parent=dyn_jean_flares_r_base, loc=<-1.3609 1.9284 -4.6322>, rot=<0.0273 -0.0978 -0.0086 -0.9948>}
 dyn_jean_flares_inner_r_01 {parent=dyn_jean_flares_inner_r_base, loc=<0 0 0>, rot=<0 0 0 -1>}
 dyn_jean_flares_inner_r_02 {parent=dyn_jean_flares_inner_r_01, loc=<7.428 0 0>, rot=<0 0 0 1>}
 dyn_jean_flares_outer_r_base {parent=dyn_jean_flares_r_base, loc=<-0.7493 1.5974 5.6516>, rot=<0.0509 0.059 0.0453 -0.9959>}
 dyn_jean_flares_outer_r_01 {parent=dyn_jean_flares_outer_r_base, loc=<0 0 0>, rot=<0 0 0 1>}
 dyn_jean_flares_outer_r_02 {parent=dyn_jean_flares_outer_r_01, loc=<7.428 0 0>, rot=<0 0 0 1>}
 deform_knee_in_base_r {parent=calf_r, loc=<-0.8031 5.0231 -0.2457>, rot=<0 0 0 -1>}
 deform_knee_in_r {parent=deform_knee_in_base_r, loc=<0 0 0>, rot=<0 0 0 1>}
 deform_knee_out_base_r {parent=calf_r, loc=<0 -4 0>, rot=<0 0 0 -1>}
 deform_knee_out_r {parent=deform_knee_out_base_r, loc=<0 0 0>, rot=<0 0 0 1>}
 thigh_twist_01_r {parent=thigh_r, loc=<27.952 0 0>, rot=<0.0013 0 0 -1>}
 deform_groin_base_r {parent=thigh_r, loc=<-0.4801 -5.6776 0.2486>, rot=<0 0 0 -1>}
 deform_groin_r {parent=deform_groin_base_r, loc=<0 0 0>, rot=<0 0 0 1>}
 deform_glute_base_l {parent=pelvis, loc=<-7.3466 -9.0213 -9.1245>, rot=<0.0172 0.031 -0.4477 -0.8935>}
 deform_glute_l {parent=deform_glute_base_l, loc=<0 0 0>, rot=<0 0 0 1>}
 deform_glute_base_r {parent=pelvis, loc=<-7.3464 -9.0213 9.1245>, rot=<0.031 -0.0172 0.8935 -0.4477>}
 deform_glute_r {parent=deform_glute_base_r, loc=<0 0 0>, rot=<0 0 0 1>}
 deform_belt_01 {parent=pelvis, loc=<1.0644 10.5902 0.9807>, rot=<-0.0265 0.0097 0.0353 -0.999>}
 deform_belt_02 {parent=pelvis, loc=<2.498 9.6387 -6.5908>, rot=<0.117 0.058 0.0554 -0.9899>}
 deform_belt_03 {parent=pelvis, loc=<3.8489 0.4915 -14.3792>, rot=<0.6883 0.0669 0.1213 -0.7121>}
 deform_belt_04 {parent=pelvis, loc=<3.3398 -7.5335 -10.8872>, rot=<0.8991 0.2139 0.1215 -0.362>}
 deform_belt_05 {parent=pelvis, loc=<2.01 -11.1543 -1.1821>, rot=<0.9705 0.2297 0.0662 -0.0316>}
 deform_belt_06 {parent=pelvis, loc=<0.2436 -10.285 10.0067>, rot=<0.9555 0.1866 -0.002 0.2287>}
 deform_belt_07 {parent=pelvis, loc=<-0.734 -3.3891 15.3055>, rot=<-0.8122 -0.1251 0.0595 -0.5667>}
 deform_belt_08 {parent=pelvis, loc=<-0.6775 4.9683 14.8548>, rot=<-0.5627 -0.0767 0.0991 -0.8171>}
 deform_gun {parent=deform_belt_08, loc=<-6.3595 1.4537 -0.0055>, rot=<0.0127 0.0009 -0.0724 -0.9973>}
 deform_belt_09 {parent=pelvis, loc=<0.1702 9.3004 9.136>, rot=<-0.0961 -0.0139 0.0534 -0.9938>}
]
variables [
 Parachute {type=bool, default=False}
 PlayerSpeed {type=double, default=0}
 "Skydive Diving  Down" {type=bool, default=False}
 Skydiving {type=bool, default=False}
]
