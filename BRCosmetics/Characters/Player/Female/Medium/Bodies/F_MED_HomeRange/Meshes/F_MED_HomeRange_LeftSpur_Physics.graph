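# Physics asset for the left spur on the HomeRange medium female player body cosmetic.
# A sphere body on the spur bone and a small capsule on its base bone are joined by one constraint.
# Swing limits are zero and twist is free, so the spur can only spin about the twist axis relative to its base.
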PhysicsAsset /BRCosmetics/Characters/Player/Female/Medium/Bodies/F_MED_HomeRange/Meshes/F_MED_HomeRange_LeftSpur_Physics
bodies {
 dyn_spur_l [{shapeType=Sphere, radius=5.0781, shapeName=dyn_spur_l_sphere, center=<0 1 0>}]
 dyn_spur_l_base [{shapeType=Capsule, length=0.505, radius=0.505, shapeName=dyn_spur_l_base_capsule, rotation=<rot 0 -90 0>}]
}
constraints [{bone1Name=dyn_spur_l, bone2Name=dyn_spur_l_base, swing1LimitDegrees=0, swing2LimitDegrees=0, twistLimitDegrees=45, twistMotion=Free}]
