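# Post-pose AnimBP for the female HomeRange cosmetic body, adding secondary motion to spurs, flared jeans, collar and necklace.
# The attached parent's pose feeds a Control Rig with skydive and parachute inputs, then per-foot RigidBody spur sims and chained spring-based AnimDynamics.
# A DefaultSlot layer filtered at deform_gun blends last, weighted by Custom Idle Is Playing.
# In the front end, updates set that weight from Frontend_Idle_Bypass and assign frontend jean/neck sim-space variables.
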AnimBlueprint /BRCosmetics/Characters/Player/Female/Medium/Bodies/F_MED_HomeRange/Meshes/F_MED_HomeRange_AnimBP
target_skeleton=&/F_MED_HomeRange_Skeleton
graphs [
 AnimGraph {nodes [
   n1 = Control Rig!(
    Source=Copy Pose From Mesh() {bUseAttachedParent=true}
    Alpha
    Skydiving:bool=Get bIsSkyDiving()
    "Skydive Diving  Down":bool=Get bIsSkydiveDiveMode()
    Parachute:bool=Get bIsParachuteOpened()
   ) -> (Pose) {
    alphaBoolBlend {blendOption=Linear}
    controlRigAssetReference {blueprintRigClass=&/F_MED_HomeRange_CtrlRig.F_MED_HomeRange_CtrlRig_C}
   }
   n2 = RigidBody(ComponentPose=Local To Component(LocalPose=@n1)) {
    overridePhysicsAsset=&/F_MED_HomeRange_LeftSpur_Physics
    simulationSpace=BaseBoneSpace
    alphaBoolBlend {blendOption=Linear}
    baseBoneRef {boneName=foot_l}
    simSpaceSettings {dampingAlpha=0, maxAngularAcceleration=300, maxAngularVelocity=1000, maxLinearAcceleration=200, maxLinearVelocity=1200, worldAlpha=1}
   }
   n3 = RigidBody(ComponentPose=@n2) {
    overridePhysicsAsset=&/F_MED_HomeRange_RightSpur_Physics
    simulationSpace=BaseBoneSpace
    alphaBoolBlend {blendOption=Linear}
    baseBoneRef {boneName=foot_r}
    simSpaceSettings {dampingAlpha=0, maxAngularAcceleration=300, maxAngularVelocity=1000, maxLinearAcceleration=200, maxLinearVelocity=1200, worldAlpha=1}
   }
   n4 = Anim Dynamics - Bone: dyn_jean_flares_front_l_01!(
    ComponentPose=@n3
    Alpha=1
    SimSpaceSettings:Anim Phys Sim Space Settings Structure=(SimSpaceAngularAlpha=0,MaxAngularVelocity=10000,MaxAngularAcceleration=10000,ExternalAngularVelocity=(X=0,Y=0,Z=0))
    ExternalForce="0,0,0"
   ) -> (Pose) {
    angularDampingOverride=0.92
    angularSpringConstant=1200
    bAngularSpring=true
    bOverrideAngularDamping=true
    bUseGravityOverride=true
    gravityScale=0.35
    simulationSpace=RootRelative
    alphaBoolBlend {blendOption=Linear}
    boundBone {boneName=dyn_jean_flares_front_l_01}
    componentAppliedLinearAccClamp {x=0, z=0}
    gravityOverride {z=-980}
    physicsBodyDefinitions [{
      boundBone {boneName=dyn_jean_flares_front_l_01}
      boxExtents=<15 25 25>
      constraintSetup {angularLimitsMax=<2 2 5>, angularLimitsMin=<-2 -2 0>, angularTarget=<1 0 0>}
      localJointOffset=<-8 0 0>
     }]
   }
   n5 = Anim Dynamics - Bone: dyn_jean_flares_front_l_02!(
    ComponentPose=@n4
    Alpha=1
    SimSpaceSettings:Anim Phys Sim Space Settings Structure=(SimSpaceAngularAlpha=1,MaxAngularVelocity=770,MaxAngularAcceleration=300,ExternalAngularVelocity=(X=0,Y=0,Z=0))
    ExternalForce="0,0,0"
   ) -> (Pose) {
    angularDampingOverride=0.92
    angularSpringConstant=1200
    bAngularSpring=true
    bOverrideAngularDamping=true
    bUseGravityOverride=true
    gravityScale=0.35
    simulationSpace=RootRelative
    alphaBoolBlend {blendOption=Linear}
    boundBone {boneName=dyn_jean_flares_front_l_02}
    componentAppliedLinearAccClamp {x=0, z=0}
    gravityOverride {z=-980}
    physicsBodyDefinitions [{
      boundBone {boneName=dyn_jean_flares_front_l_02}
      boxExtents=<15 25 25>
      constraintSetup {angularLimitsMax=<2 2 5>, angularLimitsMin=<-2 -2 0>, angularTarget=<1 0 0>}
      localJointOffset=<-8 0 0>
     }]
    simSpaceSettings {maxAngularAcceleration=300, maxAngularVelocity=770, simSpaceAngularAlpha=1}
   }
   n6 = Anim Dynamics - Bone: dyn_jean_flares_inner_l_01!(
    ComponentPose=@n5
    Alpha=1
    SimSpaceSettings:Anim Phys Sim Space Settings Structure=(SimSpaceAngularAlpha=1,MaxAngularVelocity=770,MaxAngularAcceleration=300,ExternalAngularVelocity=(X=0,Y=0,Z=0))
    ExternalForce="0,0,0"
   ) -> (Pose) {
    angularDampingOverride=0.92
    angularSpringConstant=1200
    bAngularSpring=true
    bOverrideAngularDamping=true
    bUseGravityOverride=true
    gravityScale=0.35
    simulationSpace=RootRelative
    alphaBoolBlend {blendOption=Linear}
    boundBone {boneName=dyn_jean_flares_inner_l_01}
    componentAppliedLinearAccClamp {x=0, z=0}
    gravityOverride {z=-980}
    physicsBodyDefinitions [{
      boundBone {boneName=dyn_jean_flares_inner_l_01}
      boxExtents=<15 25 25>
      constraintSetup {angularLimitsMax=<5 5 10>, angularLimitsMin=<-5 -5 0>, angularTarget=<1 0 0>}
      localJointOffset=<-8 0 0>
     }]
    simSpaceSettings {maxAngularAcceleration=300, maxAngularVelocity=770, simSpaceAngularAlpha=1}
   }
   n7 = Anim Dynamics - Bone: dyn_jean_flares_inner_l_02!(
    ComponentPose=@n6
    Alpha=1
    SimSpaceSettings:Anim Phys Sim Space Settings Structure=(SimSpaceAngularAlpha=1,MaxAngularVelocity=770,MaxAngularAcceleration=300,ExternalAngularVelocity=(X=0,Y=0,Z=0))
    ExternalForce="0,0,0"
   ) -> (Pose) {
    angularDampingOverride=0.92
    angularSpringConstant=1200
    bAngularSpring=true
    bOverrideAngularDamping=true
    bUseGravityOverride=true
    gravityScale=0.35
    simulationSpace=RootRelative
    alphaBoolBlend {blendOption=Linear}
    boundBone {boneName=dyn_jean_flares_inner_l_02}
    componentAppliedLinearAccClamp {x=0, z=0}
    gravityOverride {z=-980}
    physicsBodyDefinitions [{
      boundBone {boneName=dyn_jean_flares_inner_l_02}
      boxExtents=<15 25 25>
      constraintSetup {angularLimitsMax=<5 5 10>, angularLimitsMin=<-5 -5 0>, angularTarget=<1 0 0>}
      localJointOffset=<-8 0 0>
     }]
    simSpaceSettings {maxAngularAcceleration=300, maxAngularVelocity=770, simSpaceAngularAlpha=1}
   }
   n8 = Anim Dynamics - Bone: dyn_jean_flares_outer_l_01!(
    ComponentPose=@n7
    Alpha=1
    SimSpaceSettings:Anim Phys Sim Space Settings Structure=(SimSpaceAngularAlpha=1,MaxAngularVelocity=770,MaxAngularAcceleration=300,ExternalAngularVelocity=(X=0,Y=0,Z=0))
    ExternalForce="0,0,0"
   ) -> (Pose) {
    angularDampingOverride=0.92
    angularSpringConstant=1200
    bAngularSpring=true
    bOverrideAngularDamping=true
    bUseGravityOverride=true
    gravityScale=0.35
    simulationSpace=RootRelative
    alphaBoolBlend {blendOption=Linear}
    boundBone {boneName=dyn_jean_flares_outer_l_01}
    componentAppliedLinearAccClamp {x=0, z=0}
    gravityOverride {z=-980}
    physicsBodyDefinitions [{
      boundBone {boneName=dyn_jean_flares_outer_l_01}
      boxExtents=<15 25 25>
      constraintSetup {angularLimitsMax=<5 5 10>, angularLimitsMin=<-5 -5 0>, angularTarget=<1 0 0>}
      localJointOffset=<-8 0 0>
     }]
    simSpaceSettings {maxAngularAcceleration=300, maxAngularVelocity=770, simSpaceAngularAlpha=1}
   }
   n9 = Anim Dynamics - Bone: dyn_jean_flares_outer_l_02!(
    ComponentPose=@n8
    Alpha=1
    SimSpaceSettings:Anim Phys Sim Space Settings Structure=(SimSpaceAngularAlpha=1,MaxAngularVelocity=770,MaxAngularAcceleration=300,ExternalAngularVelocity=(X=0,Y=0,Z=0))
    ExternalForce="0,0,0"
   ) -> (Pose) {
    angularDampingOverride=0.92
    angularSpringConstant=1200
    bAngularSpring=true
    bOverrideAngularDamping=true
    bUseGravityOverride=true
    gravityScale=0.35
    simulationSpace=RootRelative
    alphaBoolBlend {blendOption=Linear}
    boundBone {boneName=dyn_jean_flares_outer_l_02}
    componentAppliedLinearAccClamp {x=0, z=0}
    gravityOverride {z=-980}
    physicsBodyDefinitions [{
      boundBone {boneName=dyn_jean_flares_outer_l_02}
      boxExtents=<15 25 25>
      constraintSetup {angularLimitsMax=<5 5 10>, angularLimitsMin=<-5 -5 0>, angularTarget=<1 0 0>}
      localJointOffset=<-8 0 0>
     }]
    simSpaceSettings {maxAngularAcceleration=300, maxAngularVelocity=770, simSpaceAngularAlpha=1}
   }
   n10 = Anim Dynamics - Bone: dyn_jean_flares_front_r_01!(
    ComponentPose=@n9
    Alpha=1
    SimSpaceSettings:Anim Phys Sim Space Settings Structure=(SimSpaceAngularAlpha=1,MaxAngularVelocity=770,MaxAngularAcceleration=300,ExternalAngularVelocity=(X=0,Y=0,Z=0))
    ExternalForce="0,0,0"
   ) -> (Pose) {
    angularDampingOverride=0.92
    angularSpringConstant=1200
    bAngularSpring=true
    bOverrideAngularDamping=true
    bUseGravityOverride=true
    gravityScale=0.35
    simulationSpace=RootRelative
    alphaBoolBlend {blendOption=Linear}
    boundBone {boneName=dyn_jean_flares_front_r_01}
    componentAppliedLinearAccClamp {x=0, z=0}
    gravityOverride {z=-980}
    physicsBodyDefinitions [{
      boundBone {boneName=dyn_jean_flares_front_r_01}
      boxExtents=<15 25 25>
      constraintSetup {angularLimitsMax=<2 2 5>, angularLimitsMin=<-2 -2 0>, angularTarget=<1 0 0>}
      localJointOffset=<8 0 0>
     }]
    simSpaceSettings {maxAngularAcceleration=300, maxAngularVelocity=770, simSpaceAngularAlpha=1}
   }
   n11 = Anim Dynamics - Bone: dyn_jean_flares_front_r_02!(
    ComponentPose=@n10
    Alpha=1
    SimSpaceSettings:Anim Phys Sim Space Settings Structure=(SimSpaceAngularAlpha=1,MaxAngularVelocity=770,MaxAngularAcceleration=300,ExternalAngularVelocity=(X=0,Y=0,Z=0))
    ExternalForce="0,0,0"
   ) -> (Pose) {
    angularDampingOverride=0.92
    angularSpringConstant=1200
    bAngularSpring=true
    bOverrideAngularDamping=true
    bUseGravityOverride=true
    gravityScale=0.35
    simulationSpace=RootRelative
    alphaBoolBlend {blendOption=Linear}
    boundBone {boneName=dyn_jean_flares_front_r_02}
    componentAppliedLinearAccClamp {x=0, z=0}
    gravityOverride {z=-980}
    physicsBodyDefinitions [{
      boundBone {boneName=dyn_jean_flares_front_r_02}
      boxExtents=<15 25 25>
      constraintSetup {angularLimitsMax=<2 2 5>, angularLimitsMin=<-2 -2 0>, angularTarget=<1 0 0>}
      localJointOffset=<8 0 0>
     }]
    simSpaceSettings {maxAngularAcceleration=300, maxAngularVelocity=770, simSpaceAngularAlpha=1}
   }
   n12 = Anim Dynamics - Bone: dyn_jean_flares_inner_r_01!(
    ComponentPose=@n11
    Alpha=1
    SimSpaceSettings:Anim Phys Sim Space Settings Structure=(SimSpaceAngularAlpha=1,MaxAngularVelocity=770,MaxAngularAcceleration=300,ExternalAngularVelocity=(X=0,Y=0,Z=0))
    ExternalForce="0,0,0"
   ) -> (Pose) {
    angularDampingOverride=0.92
    angularSpringConstant=1200
    bAngularSpring=true
    bOverrideAngularDamping=true
    bUseGravityOverride=true
    gravityScale=0.35
    simulationSpace=RootRelative
    alphaBoolBlend {blendOption=Linear}
    boundBone {boneName=dyn_jean_flares_inner_r_01}
    componentAppliedLinearAccClamp {x=0, z=0}
    gravityOverride {z=-980}
    physicsBodyDefinitions [{
      boundBone {boneName=dyn_jean_flares_inner_r_01}
      boxExtents=<15 25 25>
      constraintSetup {angularLimitsMax=<5 5 10>, angularLimitsMin=<-5 -5 0>, angularTarget=<1 0 0>}
      localJointOffset=<8 0 0>
     }]
    simSpaceSettings {maxAngularAcceleration=300, maxAngularVelocity=770, simSpaceAngularAlpha=1}
   }
   n13 = Anim Dynamics - Bone: dyn_jean_flares_inner_r_02!(
    ComponentPose=@n12
    Alpha=1
    SimSpaceSettings:Anim Phys Sim Space Settings Structure=(SimSpaceAngularAlpha=1,MaxAngularVelocity=770,MaxAngularAcceleration=300,ExternalAngularVelocity=(X=0,Y=0,Z=0))
    ExternalForce="0,0,0"
   ) -> (Pose) {
    angularDampingOverride=0.92
    angularSpringConstant=1200
    bAngularSpring=true
    bOverrideAngularDamping=true
    bUseGravityOverride=true
    gravityScale=0.35
    simulationSpace=RootRelative
    alphaBoolBlend {blendOption=Linear}
    boundBone {boneName=dyn_jean_flares_inner_r_02}
    componentAppliedLinearAccClamp {x=0, z=0}
    gravityOverride {z=-980}
    physicsBodyDefinitions [{
      boundBone {boneName=dyn_jean_flares_inner_r_02}
      boxExtents=<15 25 25>
      constraintSetup {angularLimitsMax=<5 5 10>, angularLimitsMin=<-5 -5 0>, angularTarget=<1 0 0>}
      localJointOffset=<8 0 0>
     }]
    simSpaceSettings {maxAngularAcceleration=300, maxAngularVelocity=770, simSpaceAngularAlpha=1}
   }
   n14 = Anim Dynamics - Bone: dyn_jean_flares_outer_r_01!(
    ComponentPose=@n13
    Alpha=1
    SimSpaceSettings:Anim Phys Sim Space Settings Structure=(SimSpaceAngularAlpha=1,MaxAngularVelocity=770,MaxAngularAcceleration=300,ExternalAngularVelocity=(X=0,Y=0,Z=0))
    ExternalForce="0,0,0"
   ) -> (Pose) {
    angularDampingOverride=0.92
    angularSpringConstant=1200
    bAngularSpring=true
    bOverrideAngularDamping=true
    bUseGravityOverride=true
    gravityScale=0.35
    simulationSpace=RootRelative
    alphaBoolBlend {blendOption=Linear}
    boundBone {boneName=dyn_jean_flares_outer_r_01}
    componentAppliedLinearAccClamp {x=0, z=0}
    gravityOverride {z=-980}
    physicsBodyDefinitions [{
      boundBone {boneName=dyn_jean_flares_outer_r_01}
      boxExtents=<15 25 25>
      constraintSetup {angularLimitsMax=<5 5 10>, angularLimitsMin=<-5 -5 0>, angularTarget=<1 0 0>}
      localJointOffset=<8 0 0>
     }]
    simSpaceSettings {maxAngularAcceleration=300, maxAngularVelocity=770, simSpaceAngularAlpha=1}
   }
   n15 = Anim Dynamics - Bone: dyn_jean_flares_outer_r_02!(
    ComponentPose=@n14
    Alpha=1
    SimSpaceSettings:Anim Phys Sim Space Settings Structure=(SimSpaceAngularAlpha=1,MaxAngularVelocity=770,MaxAngularAcceleration=300,ExternalAngularVelocity=(X=0,Y=0,Z=0))
    ExternalForce="0,0,0"
   ) -> (Pose) {
    angularDampingOverride=0.92
    angularSpringConstant=1200
    bAngularSpring=true
    bOverrideAngularDamping=true
    bUseGravityOverride=true
    gravityScale=0.35
    simulationSpace=RootRelative
    alphaBoolBlend {blendOption=Linear}
    boundBone {boneName=dyn_jean_flares_outer_r_02}
    componentAppliedLinearAccClamp {x=0, z=0}
    gravityOverride {z=-980}
    physicsBodyDefinitions [{
      boundBone {boneName=dyn_jean_flares_outer_r_02}
      boxExtents=<15 25 25>
      constraintSetup {angularLimitsMax=<5 5 10>, angularLimitsMin=<-5 -5 0>, angularTarget=<1 0 0>}
      localJointOffset=<8 0 0>
     }]
    simSpaceSettings {maxAngularAcceleration=300, maxAngularVelocity=770, simSpaceAngularAlpha=1}
   }
   n16 = Anim Dynamics - Bone: dyn_collar_r!(
    ComponentPose=@n15
    Alpha=1
    SimSpaceSettings:Anim Phys Sim Space Settings Structure=(SimSpaceAngularAlpha=1,MaxAngularVelocity=770,MaxAngularAcceleration=300,ExternalAngularVelocity=(X=0,Y=0,Z=0))
    ExternalForce="0,0,0"
   ) -> (Pose) {
    angularDampingOverride=0.98
    angularSpringConstant=20000
    bAngularSpring=true
    bOverrideAngularDamping=true
    bUseGravityOverride=true
    gravityScale=0.35
    simulationSpace=RootRelative
    alphaBoolBlend {blendOption=Linear}
    boundBone {boneName=dyn_collar_r}
    componentAppliedLinearAccClamp=<0 0 0>
    gravityOverride {z=-980}
    physicsBodyDefinitions [{
      boundBone {boneName=dyn_collar_r}
      boxExtents=<30 50 50>
      constraintSetup {angularLimitsMax=<8 8 8>, angularLimitsMin=<0 0 -8>, angularTarget=<1 0 0>}
      localJointOffset=<-5 0 0>
     }]
    simSpaceSettings {maxAngularAcceleration=300, maxAngularVelocity=770, simSpaceAngularAlpha=1}
   }
   n17 = Anim Dynamics - Bone: dyn_collar_l!(
    ComponentPose=@n16
    Alpha=1
    SimSpaceSettings:Anim Phys Sim Space Settings Structure=(SimSpaceAngularAlpha=1,MaxAngularVelocity=770,MaxAngularAcceleration=300,ExternalAngularVelocity=(X=0,Y=0,Z=0))
    ExternalForce="0,0,0"
   ) -> (Pose) {
    angularDampingOverride=0.98
    angularSpringConstant=20000
    bAngularSpring=true
    bOverrideAngularDamping=true
    bUseGravityOverride=true
    gravityScale=0.35
    simulationSpace=RootRelative
    alphaBoolBlend {blendOption=Linear}
    boundBone {boneName=dyn_collar_l}
    componentAppliedLinearAccClamp=<0 0 0>
    gravityOverride {z=-980}
    physicsBodyDefinitions [{
      boundBone {boneName=dyn_collar_l}
      boxExtents=<30 50 50>
      constraintSetup {angularLimitsMax=<8 8 8>, angularLimitsMin=<0 0 -8>, angularTarget=<1 0 0>}
      localJointOffset=<5 0 0>
     }]
    simSpaceSettings {maxAngularAcceleration=300, maxAngularVelocity=770, simSpaceAngularAlpha=1}
   }
   n18 = Anim Dynamics - Bone: dyn_necklace_01!(
    ComponentPose=@n17
    Alpha=1
    SimSpaceSettings:Anim Phys Sim Space Settings Structure=(SimSpaceAngularAlpha=0,MaxAngularVelocity=10000,MaxAngularAcceleration=10000,ExternalAngularVelocity=(X=0,Y=0,Z=0))
    ExternalForce="0,0,0"
   ) -> (Pose) {
    angularDampingOverride=1
    angularSpringConstant=10000
    bAngularSpring=true
    bOverrideAngularDamping=true
    bUseGravityOverride=true
    gravityScale=0.35
    simulationSpace=RootRelative
    alphaBoolBlend {blendOption=Linear}
    boundBone {boneName=dyn_necklace_01}
    componentAppliedLinearAccClamp=<0 0 0>
    gravityOverride {z=-980}
    physicsBodyDefinitions [{
      boundBone {boneName=dyn_necklace_01}
      boxExtents=<20 30 30>
      constraintSetup {angularLimitsMax=<1 5 0>, angularLimitsMin=<-1 -5 -15>, angularTarget=<1 0 0>}
      localJointOffset=<4 0 0>
     }]
   }
   n19 = Anim Dynamics - Bone: dyn_necklace_02!(
    ComponentPose=@n18
    Alpha=1
    SimSpaceSettings:Anim Phys Sim Space Settings Structure=(SimSpaceAngularAlpha=1,MaxAngularVelocity=1400,MaxAngularAcceleration=300,ExternalAngularVelocity=(X=0,Y=0,Z=0))
    ExternalForce="0,0,0"
   ) -> (Pose) {
    angularDampingOverride=1
    angularSpringConstant=10000
    bAngularSpring=true
    bOverrideAngularDamping=true
    bUseGravityOverride=true
    gravityScale=0.35
    simulationSpace=RootRelative
    alphaBoolBlend {blendOption=Linear}
    boundBone {boneName=dyn_necklace_02}
    componentAppliedLinearAccClamp=<0 0 0>
    gravityOverride {z=-980}
    physicsBodyDefinitions [{
      boundBone {boneName=dyn_necklace_02}
      boxExtents=<20 30 30>
      constraintSetup {angularLimitsMax=<1 5 0>, angularLimitsMin=<-1 -5 -15>, angularTarget=<1 0 0>}
      localJointOffset=<4 0 0>
     }]
    simSpaceSettings {maxAngularAcceleration=300, maxAngularVelocity=1400, simSpaceAngularAlpha=1}
   }
   n20 = Anim Dynamics - Bone: dyn_necklace_03!(
    ComponentPose=@n19
    Alpha=1
    SimSpaceSettings:Anim Phys Sim Space Settings Structure=(SimSpaceAngularAlpha=1,MaxAngularVelocity=1400,MaxAngularAcceleration=300,ExternalAngularVelocity=(X=0,Y=0,Z=0))
    ExternalForce="0,0,0"
   ) -> (Pose) {
    angularDampingOverride=1
    angularSpringConstant=10000
    bAngularSpring=true
    bOverrideAngularDamping=true
    bUseGravityOverride=true
    gravityScale=0.35
    simulationSpace=RootRelative
    alphaBoolBlend {blendOption=Linear}
    boundBone {boneName=dyn_necklace_03}
    componentAppliedLinearAccClamp=<0 0 0>
    gravityOverride {z=-980}
    physicsBodyDefinitions [{
      boundBone {boneName=dyn_necklace_03}
      boxExtents=<20 30 30>
      constraintSetup {angularLimitsMax=<1 5 0>, angularLimitsMin=<-1 -5 -15>, angularTarget=<1 0 0>}
      localJointOffset=<4 0 0>
     }]
    simSpaceSettings {maxAngularAcceleration=300, maxAngularVelocity=1400, simSpaceAngularAlpha=1}
   }
   n21 = Anim Dynamics - Bone: dyn_necklace_04!(
    ComponentPose=@n20
    Alpha=1
    SimSpaceSettings:Anim Phys Sim Space Settings Structure=(SimSpaceAngularAlpha=1,MaxAngularVelocity=1400,MaxAngularAcceleration=300,ExternalAngularVelocity=(X=0,Y=0,Z=0))
    ExternalForce="0,0,0"
   ) -> (Pose) {
    angularDampingOverride=1
    angularSpringConstant=10000
    bAngularSpring=true
    bOverrideAngularDamping=true
    bUseGravityOverride=true
    gravityScale=0.35
    simulationSpace=RootRelative
    alphaBoolBlend {blendOption=Linear}
    boundBone {boneName=dyn_necklace_04}
    componentAppliedLinearAccClamp=<0 0 0>
    gravityOverride {z=-980}
    physicsBodyDefinitions [{
      boundBone {boneName=dyn_necklace_04}
      boxExtents=<20 30 30>
      constraintSetup {angularLimitsMax=<5 5 0>, angularLimitsMin=<-5 -5 -15>, angularTarget=<1 0 0>}
      localJointOffset=<4 0 0>
     }]
    simSpaceSettings {maxAngularAcceleration=300, maxAngularVelocity=1400, simSpaceAngularAlpha=1}
   }
   n22 = Anim Dynamics - Bone: dyn_necklace_05!(
    ComponentPose=@n21
    Alpha=1
    SimSpaceSettings:Anim Phys Sim Space Settings Structure=(SimSpaceAngularAlpha=1,MaxAngularVelocity=1400,MaxAngularAcceleration=300,ExternalAngularVelocity=(X=0,Y=0,Z=0))
    ExternalForce="0,0,0"
   ) -> (Pose) {
    angularDampingOverride=1
    angularSpringConstant=100
    bAngularSpring=true
    bOverrideAngularDamping=true
    bUseGravityOverride=true
    gravityScale=0.35
    simulationSpace=RootRelative
    alphaBoolBlend {blendOption=Linear}
    boundBone {boneName=dyn_necklace_05}
    componentAppliedLinearAccClamp=<0 0 0>
    gravityOverride {z=-980}
    physicsBodyDefinitions [{
      boundBone {boneName=dyn_necklace_05}
      boxExtents=<10 5 5>
      constraintSetup {angularLimitsMax=<20 20 0>, angularLimitsMin=<-20 -20 -30>, angularTarget=<1 0 0>}
      localJointOffset=<4 0 0>
     }]
    simSpaceSettings {maxAngularAcceleration=300, maxAngularVelocity=1400, simSpaceAngularAlpha=1}
   }
   Output Pose(Result=Layered blend per bone(
     BasePose=Component To Local(ComponentPose=@n22)
     BlendPoses_0=Slot 'DefaultSlot'()
     BlendWeights_0=Get Custom Idle Is Playing() -> (Custom Idle Is Playing:double)
    ) {layerSetup [{branchFilters [{boneName=deform_gun}]}]}) {layerGroup=None}
  ]}
 EventGraph {nodes [
   Event BlueprintUpdateAnimation()
   > Branch(Condition=Get bIsInFrontEnd())
   > Set Neck Sim Space Settings(Neck Sim Space Settings:Anim Phys Sim Space Settings Structure=Get Neck Sim Space Settings Frontend() -> (Neck Sim Space Settings Frontend:Anim Phys Sim Space Settings Structure)) -> (Output_Get:Anim Phys Sim Space Settings Structure)
   > Set Jean Sim Space Settings(Jean Sim Space Settings:Anim Phys Sim Space Settings Structure=Get Jean Sim Space Settings Frontend() -> (Jean Sim Space Settings Frontend:Anim Phys Sim Space Settings Structure)) -> (Output_Get:Anim Phys Sim Space Settings Structure)
   > Set Custom Idle Is Playing(Custom Idle Is Playing:double=GetCurveValue(self=GetOwnerFrontEndAnimInstance(), CurveName=Frontend_Idle_Bypass)) -> (Output_Get:double)
  ]}
]
variables [
 Custom Idle Is Playing {type=number}
 Jean Sim Space Settings {
  type=AnimPhysSimSpaceSettings
  default {maxAngularAcceleration=300, maxAngularVelocity=770, simSpaceAngularAlpha=1, externalAngularVelocity=<0 0 0>}
 }
 Jean Sim Space Settings Default {
  type=AnimPhysSimSpaceSettings
  default {maxAngularAcceleration=300, maxAngularVelocity=770, simSpaceAngularAlpha=1, externalAngularVelocity=<0 0 0>}
 }
 Jean Sim Space Settings Frontend {
  type=AnimPhysSimSpaceSettings
  default {maxAngularAcceleration=100, maxAngularVelocity=300, simSpaceAngularAlpha=0.5, externalAngularVelocity=<0 0 0>}
 }
 Neck Sim Space Settings {
  type=AnimPhysSimSpaceSettings
  default {maxAngularAcceleration=300, maxAngularVelocity=1400, simSpaceAngularAlpha=1, externalAngularVelocity=<0 0 0>}
 }
 Neck Sim Space Settings Default {
  type=AnimPhysSimSpaceSettings
  default {maxAngularAcceleration=300, maxAngularVelocity=1400, simSpaceAngularAlpha=1, externalAngularVelocity=<0 0 0>}
 }
 Neck Sim Space Settings Frontend {
  type=AnimPhysSimSpaceSettings
  default {maxAngularAcceleration=50, maxAngularVelocity=200, simSpaceAngularAlpha=0.7, externalAngularVelocity=<0 0 0>}
 }
]
