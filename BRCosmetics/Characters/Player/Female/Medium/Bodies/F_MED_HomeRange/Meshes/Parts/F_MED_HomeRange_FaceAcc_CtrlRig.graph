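# Control Rig for the F_MED_HomeRange character's face accessories, deforming the hat, hair volumes and hat strings as the head moves.
# Deform Hair and Hat aims rear hair at spine_05, then spherical pose readers on head rotation weight corrective bone offsets, one scaled by isOnGround.
# The rear-bend reader also writes the "SPR Head Rot Back" curve.
# Deform Hat Strings then aims the string base at the rear hat lip and applies curve-weighted offsets.
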ControlRigBlueprint /BRCosmetics/Characters/Player/Female/Medium/Bodies/F_MED_HomeRange/Meshes/Parts/F_MED_HomeRange_FaceAcc_CtrlRig
graphs [
 New Function_ContainedGraph {nodes [
   RigVMFunctionEntryNode()
   > &RigUnit_AimConstraintLocalSpaceOffset(
    Child=(Type=Bone,Name="dyn_hatstrings_base")
    Filter=(bX=True,bY=True,bZ=True)
    Parents=((Item=(Type=Bone,Name="dyn_hat_lip_rear_02"),Weight=1))
    Weight=1
    WorldUp=(Target=(X=0,Y=0,Z=1),Kind=Location,Space=(Type=Bone,Name="dyn_hat_lip_rear_02"))
   )
   > &RigUnit_ModifyTransforms(
    ItemToModify=((Item=(Type=Bone,Name="dyn_hatstrings_base"),Transform=(Rotation=(X=0,Y=0,Z=0.061,W=0.9981),Translation=(X=0,Y=0,Z=0),Scale3D=(X=1,Y=1,Z=1))))
    Weight=1
    WeightMinimum=0
   )
   > &RigUnit_ModifyTransforms(
    ItemToModify=((Item=(Type=Bone,Name="dyn_hatstrings_base"),Transform=(Rotation=(X=0,Y=0,Z=-0.2588,W=0.9659),Translation=(X=0,Y=-0.5,Z=0),Scale3D=(X=1,Y=1,Z=1))))
    WeightMinimum=0
    Weight=/Script/ControlRig.RigUnit_GetCurveValue(Curve=SPR Head Rot Back) {is=struct RigVMUnitNode}.Value
   )
   > RigVMFunctionReturnNode()
  ]}
 New Function_ContainedGraph {nodes [
   n1 = RigVMFunctionEntryNode()
   RigVMFunction_Sequence = &RigVMFunction_Sequence(ExecuteContext=@n1.ExecuteContext)
   &RigUnit_AimConstraintLocalSpaceOffset(
    Child=(Type=Bone,Name="dyn_hair_volume_rear_base")
    Filter=(bX=True,bY=True,bZ=True)
    Parents=((Item=(Type=Bone,Name="spine_05"),Weight=1))
    Weight=1
    WorldUp=(Target=(X=0,Y=0,Z=1),Kind=Location,Space=(Type=Bone,Name="spine_05"))
    ExecutePin=@RigVMFunction_Sequence.A
   )
   > &RigUnit_AimConstraintLocalSpaceOffset(
    Child=(Type=Bone,Name="dyn_hair_volume_r_b_base")
    Filter=(bX=True,bY=True,bZ=True)
    Parents=((Item=(Type=Bone,Name="spine_05"),Weight=1))
    Weight=1
    WorldUp=(Target=(X=0,Y=0,Z=1),Kind=Location,Space=(Type=Bone,Name="spine_05"))
   )
   > &RigUnit_AimConstraintLocalSpaceOffset(
    Child=(Type=Bone,Name="dyn_hair_volume_l_b_base")
    Filter=(bX=True,bY=True,bZ=True)
    Parents=((Item=(Type=Bone,Name="spine_05"),Weight=1))
    Weight=1
    WorldUp=(Target=(X=0,Y=0,Z=1),Kind=Location,Space=(Type=Bone,Name="spine_05"))
   )
   > n2 = &RigUnit_SphericalPoseReader(
    ActiveRegionScaleFactors=(PositiveWidth=6,NegativeWidth=1,PositiveHeight=6,NegativeHeight=6)
    ActiveRegionSize=0.1
    Debug=(bDrawDebug=false,bDraw2D=False,bDrawLocalAxes=False,DebugScale=25,DebugSegments=20,DebugThickness=0.25)
    DriverAxis=(X=1,Y=0,Z=0)
    DriverItem=(Type=Bone,Name="head")
    FalloffRegionScaleFactors=(PositiveWidth=1,NegativeWidth=1,PositiveHeight=1,NegativeHeight=1)
    FalloffSize=0.6
    FlipHeightScaling=False
    FlipWidthScaling=False
    OptionalParentItem=(Type=Bone,Name="spine_05")
    RotationOffset=(X=90,Y=0,Z=160)
   )
   > /Script/ControlRig.RigUnit_SetCurveValue(Curve=SPR Head Rot Back, Value=@n2.OutputParam) {is=struct RigVMUnitNode}
   RigVMFunction_AnimEvalRichCurve = &RigVMFunction_AnimEvalRichCurve(
    Curve=(EditorCurveData=(Keys=((),(Time=0.2615,Value=0.1932),(Time=0.7549,Value=0.4702),(Time=1,Value=1))))
    SourceMinimum=0
    TargetMinimum=0
    Value=@n2.OutputParam
   )
   > &RigUnit_ModifyTransforms(
    ItemToModify=((Item=(Type=Bone,Name="deform_hat"),Transform=(Rotation=(X=0,Y=0,Z=0.4864,W=0.8738),Translation=(X=0.5859,Y=2.5418,Z=-0.0001),Scale3D=(X=1,Y=1,Z=1))))
    WeightMinimum=0
    Weight=@RigVMFunction_AnimEvalRichCurve.Result
   )
   > &RigUnit_ModifyTransforms(
    ItemToModify="((Item=(Type=Bone,Name=\"dyn_rear_rotation_base\"),Transform=(Rotation=(X=0,Y=-0.0872,Z=0,W=0.9962),Translation=(X=-5.6442,Y=0,Z=-7.0554),Scale3D=(X=1,Y=1,Z=1))),(Item=(Type=Bone,Name=\"dyn_hat_lip_rear_base\"),Transform=(Rotation=(X=0,Y=0,Z=0,W=1),Translation ...(truncated)"
    WeightMinimum=0
    Weight=@n2.OutputParam
   )
   > n3 = &RigUnit_SphericalPoseReader(
    ActiveRegionScaleFactors=(PositiveWidth=1,NegativeWidth=1,PositiveHeight=1,NegativeHeight=1)
    ActiveRegionSize=0.1
    Debug=(bDrawDebug=false,bDraw2D=False,bDrawLocalAxes=False,DebugScale=25,DebugSegments=20,DebugThickness=0.25)
    DriverAxis=(X=1,Y=0,Z=0)
    DriverItem=(Type=Bone,Name="head")
    FalloffRegionScaleFactors=(PositiveWidth=1,NegativeWidth=1,PositiveHeight=1,NegativeHeight=1)
    FalloffSize=0.3
    FlipHeightScaling=False
    FlipWidthScaling=False
    OptionalParentItem=(Type=Bone,Name="spine_05")
    RotationOffset=(X=55,Y=20,Z=200)
   )
   > &RigUnit_ModifyTransforms(
    ItemToModify="((Item=(Type=Bone,Name=\"dyn_hair_volume_r_b_base\"),Transform=(Rotation=(X=0,Y=0,Z=0,W=1),Translation=(X=0,Y=0,Z=0),Scale3D=(X=1,Y=1,Z=1))),(Item=(Type=Bone,Name=\"dyn_hat_lip_rear_base\"),Transform=(Rotation=(X=0.2261,Y=-0.1245,Z=-0.0364,W=-0.9654),Translati ...(truncated)"
    WeightMinimum=0
    Weight=@n3.OutputParam
   )
   > n4 = &RigUnit_SphericalPoseReader(
    ActiveRegionScaleFactors=(PositiveWidth=2,NegativeWidth=2,PositiveHeight=3,NegativeHeight=1)
    ActiveRegionSize=0.1
    Debug=(bDrawDebug=false,bDraw2D=False,bDrawLocalAxes=False,DebugScale=25,DebugSegments=20,DebugThickness=0.25)
    DriverAxis=(X=0,Y=1,Z=0)
    DriverItem=(Type=Bone,Name="head")
    FalloffRegionScaleFactors=(PositiveWidth=1,NegativeWidth=1,PositiveHeight=1,NegativeHeight=0.8)
    FalloffSize=0.2
    FlipHeightScaling=False
    FlipWidthScaling=False
    OptionalParentItem=(Type=Bone,Name="spine_05")
    RotationOffset=(X=125,Y=0,Z=0)
   )
   > &RigUnit_ModifyTransforms(
    ItemToModify=((Item=(Type=Bone,Name="dyn_hatstring_attach"),Transform=(Rotation=(X=0,Y=0,Z=0,W=-1),Translation=(X=0,Y=1.2087,Z=-0.721),Scale3D=(X=1,Y=1,Z=1))))
    WeightMinimum=0
    Weight=@n4.OutputParam
   )
   > &RigUnit_ModifyTransforms(
    ItemToModify="((Item=(Type=Bone,Name=\"dyn_front_hatstring_l_base\"),Transform=(Rotation=(X=0.0148,Y=0.0283,Z=0.1372,W=0.99),Translation=(X=0,Y=0,Z=0),Scale3D=(X=1,Y=1,Z=1))),(Item=(Type=Bone,Name=\"dyn_front_hatstring_r_base\"),Transform=(Rotation=(X=0,Y=0,Z=0,W=1),Transla ...(truncated)"
    WeightMinimum=0
    Weight=@n4.OutputParam
   )
   > n5 = &RigUnit_SphericalPoseReader(
    ActiveRegionScaleFactors=(PositiveWidth=1,NegativeWidth=1,PositiveHeight=1,NegativeHeight=1)
    ActiveRegionSize=0.12
    Debug=(bDrawDebug=True,bDraw2D=False,bDrawLocalAxes=False,DebugScale=25,DebugSegments=20,DebugThickness=0.25)
    DriverAxis=(X=1,Y=0,Z=0)
    DriverItem=(Type=Bone,Name="head")
    FalloffRegionScaleFactors=(PositiveWidth=1,NegativeWidth=1,PositiveHeight=1,NegativeHeight=1)
    FalloffSize=0.2
    FlipHeightScaling=False
    FlipWidthScaling=False
    OptionalParentItem=(Type=Bone,Name="None")
    RotationOffset=(X=90,Y=0,Z=-147)
   )
   Multiply = &RigVMFunction_MathDoubleMul(A=@n5.OutputParam, B=RigVMVariableNode() {variable=isOnGround}.Value)
   > &RigUnit_ModifyTransforms(
    ItemToModify="((Item=(Type=Bone,Name=\"dyn_hat_lip_rear_base\"),Transform=(Rotation=(X=-0.1239,Y=-0.1303,Z=0.0104,W=-0.9837),Translation=(X=0,Y=0,Z=0),Scale3D=(X=1,Y=1,Z=1))),(Item=(Type=Bone,Name=\"dyn_hair_volume_l_b_02\"),Transform=(Rotation=(X=0,Y=-0.0296,Z=0,W=0.9996), ...(truncated)"
    WeightMinimum=0
    Weight=@Multiply.Result
   )
   RigVMFunctionReturnNode(ExecuteContext=@RigVMFunction_Sequence.B)
   &RigUnit_ModifyTransforms(
    ExecutePin=()
    ItemToModify="((Item=(Type=Bone,Name=\"dyn_hat_lip_rear_base\"),Transform=(Rotation=(X=0.0513,Y=0.1248,Z=-0.0018,W=0.9909),Translation=(X=0,Y=0,Z=0),Scale3D=(X=1,Y=1,Z=1))),(Item=(Type=Bone,Name=\"dyn_hair_volume_l_b_02\"),Transform=(Rotation=(X=0,Y=-0.0296,Z=0,W=0.9996),Tr ...(truncated)"
    Weight=1
    WeightMinimum=0
   )
  ]}
 RigVMFunctionLibrary {nodes [
   Deform Hair and Hat = RigVMCollapseNode(ExecuteContext=())
   Deform Hat Strings = RigVMCollapseNode(ExecuteContext=())
  ]}
 RigVMModel {nodes [
   RigUnit_BeginExecution = &RigUnit_BeginExecution()
   Deform Hair and Hat = RigVMFunctionReferenceNode(ExecuteContext=@RigUnit_BeginExecution.ExecutePin)
   > Deform Hat Strings = RigVMFunctionReferenceNode()
  ]}
]
hierarchy [
 root {loc=<0 0 0>, rot=<0 0 0 1>}
 pelvis {parent=root, loc=<0 -0.5301 98.0571>, rot=<0.0127 0.707 -0.0127 -0.707>}
 spine_01 {parent=pelvis, loc=<6.1208 -0.9328 0>, rot=<0 0 0.034 -0.9994>}
 spine_02 {parent=spine_01, loc=<6.5996 -0.2095 0>, rot=<0 0 0.0238 -0.9997>}
 spine_03 {parent=spine_02, loc=<6.4914 -0.1237 0>, rot=<0 0 0.0149 -0.9999>}
 spine_04 {parent=spine_03, loc=<6.9598 0.0794 0>, rot=<0 0 -0.0132 -0.9999>}
 spine_05 {parent=spine_04, loc=<6.9987 0 0>, rot=<0 0 -0.0015 -1>}
 neck_01 {parent=spine_05, loc=<11.3138 0.1529 0>, rot=<0 0 -0.2003 -0.9797>}
 neck_02 {parent=neck_01, loc=<4.5627 0 0>, rot=<0 0 0.0504 -0.9987>}
 head {parent=neck_02, loc=<5.3914 0 0>, rot=<0 0 0.115 -0.9934>}
 deform_hat_base {parent=head, loc=<6.4853 -0.5891 0.2727>, rot=<0.0053 0.0152 0.3164 -0.9485>}
 dyn_hair_fringe_a_base {parent=deform_hat_base, loc=<-4.0704 10.8586 -0.7942>, rot=<-0.1336 -0.4893 0.7404 -0.441>}
 dyn_hair_fringe_a_01 {parent=dyn_hair_fringe_a_base, loc=<0 0 0>, rot=<0 0 0 1>}
 dyn_hair_fringe_a_02 {parent=dyn_hair_fringe_a_01, loc=<5.4341 0 0>, rot=<0.023 0.1582 -0.1745 -0.9716>}
 dyn_hair_fringe_b_base {parent=deform_hat_base, loc=<-4.2591 7.3799 -5.2322>, rot=<-0.0722 -0.2799 0.7874 -0.5444>}
 dyn_hair_fringe_b_01 {parent=dyn_hair_fringe_b_base, loc=<0 0 0>, rot=<0 0 0 1>}
 dyn_hair_fringe_b_02 {parent=dyn_hair_fringe_b_01, loc=<5.9102 1.8705 0.2423>, rot=<0.0481 -0.0084 -0.206 -0.9773>}
 dyn_hair_fringe_b_03 {parent=dyn_hair_fringe_b_01, loc=<2.4507 2.475 -0.1537>, rot=<0.0571 0.125 -0.354 -0.9251>}
 dyn_hair_fringe_b_04 {parent=dyn_hair_fringe_b_01, loc=<8.576 0.2271 0.9018>, rot=<-0.0226 0.0667 -0.4094 -0.9096>}
 dyn_hair_fringe_c_base {parent=deform_hat_base, loc=<-2.9017 9.1142 2.0754>, rot=<0.5814 -0.7054 0.34 0.221>}
 dyn_hair_fringe_c_01 {parent=dyn_hair_fringe_c_base, loc=<0 0 0>, rot=<0 0 0 1>}
 dyn_hair_fringe_c_02 {parent=dyn_hair_fringe_c_01, loc=<4.8707 0 0>, rot=<0.0534 0.0466 0.2676 -0.9609>}
 dyn_hair_fringe_c_03 {parent=dyn_hair_fringe_c_01, loc=<2.6993 2.4634 0.1492>, rot=<0.0597 -0.0651 -0.1323 -0.9873>}
 dyn_hair_volume_l_a_base {parent=deform_hat_base, loc=<-1.6295 -0.8659 -7.469>, rot=<-0.1382 0.0746 0.8498 -0.5032>}
 dyn_hair_volume_l_a_01 {parent=dyn_hair_volume_l_a_base, loc=<0 0 0>, rot=<0 0 0 1>}
 dyn_hair_volume_l_a_02 {parent=dyn_hair_volume_l_a_01, loc=<5.0535 0 0>, rot=<0 -0.012 -0.0843 -0.9964>}
 dyn_hair_volume_r_a_base {parent=deform_hat_base, loc=<-2.1302 -3.225 6.1468>, rot=<-0.0911 -0.1653 -0.5005 -0.8449>}
 dyn_hair_volume_r_a_01 {parent=dyn_hair_volume_r_a_base, loc=<0 0 0>, rot=<0 0 0 1>}
 dyn_hair_volume_r_a_02 {parent=dyn_hair_volume_r_a_01, loc=<-3.1516 0 0>, rot=<0 0 0 1>}
 dyn_rear_rotation_base {parent=deform_hat_base, loc=<-2.1846 -6.4763 -0.2559>, rot=<-0.4298 0.5315 0.5684 -0.458>}
 dyn_hair_volume_rear_base {parent=dyn_rear_rotation_base, loc=<3.2293 0 1.1705>, rot=<0 0 0 1>}
 dyn_hair_volume_rear_01 {parent=dyn_hair_volume_rear_base, loc=<0 0 0>, rot=<0 0 0 1>}
 dyn_hair_volume_rear_02 {parent=dyn_hair_volume_rear_01, loc=<5.256 0 0.0875>, rot=<0 0.0263 0 -0.9997>}
 dyn_hair_volume_l_b_base {parent=dyn_rear_rotation_base, loc=<1.392 -5.1602 0.8407>, rot=<0.171 0.0463 0.2143 -0.9606>}
 dyn_hair_volume_l_b_01 {parent=dyn_hair_volume_l_b_base, loc=<0 0 0>, rot=<0 0 0 1>}
 dyn_hair_volume_l_b_02 {parent=dyn_hair_volume_l_b_01, loc=<5.256 0 0.0875>, rot=<0 0.0263 0 -0.9997>}
 dyn_hair_volume_r_b_base {parent=dyn_rear_rotation_base, loc=<1.3087 4.9341 1.1847>, rot=<0.2205 -0.9524 -0.204 -0.0522>}
 dyn_hair_volume_r_b_01 {parent=dyn_hair_volume_r_b_base, loc=<0 0 0>, rot=<0 0 0 1>}
 dyn_hair_volume_r_b_02 {parent=dyn_hair_volume_r_b_01, loc=<-5.2565 -0.0001 -0.0879>, rot=<0 0.0263 0 -0.9997>}
 deform_hat {parent=deform_hat_base, loc=<0 0 0>, rot=<0 0 0 1>}
 dyn_hatstrings_base {parent=deform_hat, loc=<2.6526 -9.0627 -0.4145>, rot=<-0.0014 -0.0139 -0.3164 -0.9485>}
 dyn_hatstring_r_01 {parent=dyn_hatstrings_base, loc=<-0.5802 -0.0899 0.1356>, rot=<0.1177 -0.6982 -0.6449 0.2876>}
 dyn_hatstring_r_02 {parent=dyn_hatstring_r_01, loc=<4.4544 0 0>, rot=<0 -0.0638 0 -0.998>}
 dyn_hatstring_r_03 {parent=dyn_hatstring_r_02, loc=<3.8048 0 0>, rot=<-0.0904 0.3358 0.0324 -0.937>}
 dyn_hatstring_r_04 {parent=dyn_hatstring_r_03, loc=<2.7791 0 0>, rot=<-0.0232 0.0364 0.0082 -0.999>}
 dyn_hatstring_r_05 {parent=dyn_hatstring_r_04, loc=<3.5327 0 0>, rot=<0.1163 -0.0221 0.0231 -0.9927>}
 dyn_hatstring_r_06 {parent=dyn_hatstring_r_05, loc=<2.0619 0 0>, rot=<0 0 0 1>}
 dyn_hatstring_l_01 {parent=dyn_hatstrings_base, loc=<-0.5134 -0.0703 -0.0628>, rot=<0.1461 -0.7929 0.5105 -0.2991>}
 dyn_hatstring_l_02 {parent=dyn_hatstring_l_01, loc=<4.3204 0 0>, rot=<0 0.0525 0 -0.9986>}
 dyn_hatstring_l_03 {parent=dyn_hatstring_l_02, loc=<3.8849 0 0>, rot=<0.0082 -0.3255 0.0028 -0.9455>}
 dyn_hatstring_l_04 {parent=dyn_hatstring_l_03, loc=<2.7661 0 0>, rot=<-0.1514 -0.0449 0.0289 -0.987>}
 dyn_hatstring_l_05 {parent=dyn_hatstring_l_04, loc=<3.0201 0 0>, rot=<0 0 0 1>}
 dyn_hattie_01_base {parent=deform_hat, loc=<-0.554 1.1686 -0.3949>, rot=<0.0157 -0.0037 0.9766 -0.2145>}
 dyn_hattie_01 {parent=dyn_hattie_01_base, loc=<0 0 0>, rot=<0 0 0 1>}
 dyn_hattie_02 {parent=dyn_hattie_01, loc=<5.856 0 0>, rot=<0 -0.0136 0 -0.9999>}
 dyn_hattie_03 {parent=dyn_hattie_02, loc=<6.5374 0 0>, rot=<0 0 0 1>}
 dyn_hattie_04 {parent=dyn_hattie_03, loc=<4.2387 0 0>, rot=<0 0 0 1>}
 dyn_hat_lip_l_01 {parent=deform_hat, loc=<-0.3674 0.2565 -9.0491>, rot=<0.9183 -0.0323 -0.3813 -0.102>}
 dyn_hat_lip_l_02 {parent=dyn_hat_lip_l_01, loc=<3.8 0 0>, rot=<0 -0.3811 0 -0.9246>}
 dyn_hat_lip_r_01 {parent=deform_hat, loc=<0.1895 0.2498 8.2286>, rot=<-0.0291 -0.9301 0.1026 -0.3515>}
 dyn_hat_lip_r_02 {parent=dyn_hat_lip_r_01, loc=<-3.7997 -0.0004 -0.0002>, rot=<0 -0.3811 0 -0.9246>}
 dyn_hat_lip_rear_base {parent=deform_hat, loc=<-0.1487 -8.1257 -0.4117>, rot=<-0.4386 0.5402 0.5548 -0.456>}
 dyn_hat_lip_rear_01 {parent=dyn_hat_lip_rear_base, loc=<0 0 0>, rot=<0 0 0 1>}
 dyn_hat_lip_rear_02 {parent=dyn_hat_lip_rear_01, loc=<5.256 0 0.0875>, rot=<0 0.0263 0 -0.9997>}
 dyn_front_hatstring_l_base {parent=deform_hat, loc=<-1.1063 1.0251 -7.5335>, rot=<0.2231 -0.9695 0.0829 -0.0585>}
 dyn_front_hatstring_l_01 {parent=dyn_front_hatstring_l_base, loc=<0 0 0>, rot=<0 0 0 1>}
 dyn_front_hatstring_l_02 {parent=dyn_front_hatstring_l_01, loc=<6.6674 0 0>, rot=<0 -0.1866 0 -0.9824>}
 dyn_front_hatstring_l_03 {parent=dyn_front_hatstring_l_02, loc=<5.6038 0 0>, rot=<0.2356 -0.1059 0.0258 -0.9657>}
 dyn_front_hatstring_l_04 {parent=dyn_front_hatstring_l_03, loc=<4.1801 0 0>, rot=<-0.1884 -0.1178 -0.0228 -0.9747>}
 dyn_front_hatstring_l_05 {parent=dyn_front_hatstring_l_04, loc=<2.7798 0 0>, rot=<0.0028 -0.2244 0.0006 -0.9745>}
 dyn_front_hatstring_l_06 {parent=dyn_front_hatstring_l_05, loc=<3.3196 0 0>, rot=<0 0 0 1>}
 dyn_front_hatstring_r_base {parent=deform_hat, loc=<-0.6498 1.84 7.1476>, rot=<0.2517 -0.9601 -0.0793 0.0922>}
 dyn_front_hatstring_r_01 {parent=dyn_front_hatstring_r_base, loc=<0 0 0>, rot=<0 0 0 1>}
 dyn_front_hatstring_r_02 {parent=dyn_front_hatstring_r_01, loc=<6.6421 0 0>, rot=<0 0.1787 0 -0.9839>}
 dyn_front_hatstring_r_03 {parent=dyn_front_hatstring_r_02, loc=<5.6114 0 0>, rot=<-0.0346 0.1082 0.0038 -0.9935>}
 dyn_front_hatstring_r_04 {parent=dyn_front_hatstring_r_03, loc=<4.262 0 0>, rot=<-0.0608 0.1266 0.0078 -0.9901>}
 dyn_front_hatstring_r_05 {parent=dyn_front_hatstring_r_04, loc=<3.3755 0 0>, rot=<0.0616 0.2159 -0.0136 -0.9744>}
 dyn_front_hatstring_r_06 {parent=dyn_front_hatstring_r_05, loc=<2.7008 0 0>, rot=<0 0 0 1>}
 dyn_collar_l {parent=spine_05, loc=<7.344 2.6775 -6.8864>, rot=<-0.0515 0.1389 -0.2437 -0.9585>}
 dyn_collar_r {parent=spine_05, loc=<7.344 2.6775 6.8864>, rot=<0.1389 0.0515 0.9585 -0.2437>}
 clavicle_l {parent=spine_05, loc=<5.4831 -0.1053 -3.2946>, rot=<0.0465 -0.7289 -0.009 -0.6829>}
 clavicle_r {parent=spine_05, loc=<5.4831 -0.1053 3.2946>, rot=<0.7289 0.0465 -0.6829 0.009>}
 upperarm_r {parent=clavicle_r, loc=<-9.8658 1.2136 0.4105>, rot=<0.0164 -0.3245 -0.0269 -0.9454>}
 dyn_hatstring_attach {parent=spine_05, loc=<8.033 7.6115 0.2219>, rot=<0 0 0 1>}
 SPR_Head Rot Back {type=Curve, loc=<0 0 0>, rot=<0 0 0 1>}
 SPR Head Rot Back {type=Curve, loc=<0 0 0>, rot=<0 0 0 1>}
]
variables [
 SphericalPoseReader_OutputParam {type=float, default=0}
 isOnGround {type=double, default=0}
]
