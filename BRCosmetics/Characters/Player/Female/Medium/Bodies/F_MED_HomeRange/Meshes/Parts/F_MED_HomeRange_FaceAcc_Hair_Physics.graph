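# Physics asset for the HomeRange female character's face accessory and hair part, covering hair, hat, hat strings and collars.
# Mostly capsule bodies form hair chains off the head and a rear rotation base, plus string chains off the box-shaped hat, with tight angular limits between segments.
# The front string tips are locked together into a loop, and rear strings also link to a hat-lip bone that has no body here.
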PhysicsAsset /BRCosmetics/Characters/Player/Female/Medium/Bodies/F_MED_HomeRange/Meshes/Parts/F_MED_HomeRange_FaceAcc_Hair_Physics
bodies {
 deform_hat [{
   shapeType=Box
   extentX=5
   extentY=34.4828
   extentZ=32.7039
   shapeName=deform_hat_box
   center=<4.3324 -1.6014 -0.9139>
   rotation=<rot -0.8296 5.8124 -85.2822>
  }]
 dyn_collar_l [{
   shapeType=Capsule
   length=10.6504
   radius=2.7712
   shapeName=dyn_collar_l_capsule
   center=<2.2564 0.0736 0.0228>
   rotation=<rot 69.8696 -90.0002 -14.2524>
  }]
 dyn_collar_r [{
   shapeType=Capsule
   length=10.6504
   radius=2.7712
   shapeName=dyn_collar_r_capsule
   center=<-2.2564 -0.0736 -0.0228>
   rotation=<rot 69.8696 -90.0002 165.7476>
  }]
 dyn_front_hatstring_l_01 [{shapeType=Capsule, length=2.9406, radius=1, shapeName=dyn_front_hatstring_l_01_capsule, center=<2 0 0>, rotation=<rot 90 0 0>}]
 dyn_front_hatstring_l_02 [{shapeType=Capsule, length=4.5937, radius=1, shapeName=dyn_front_hatstring_l_02_capsule, center=<2 0 0>, rotation=<rot 90 0 0>}]
 dyn_front_hatstring_l_03 [{shapeType=Capsule, length=3.2028, radius=1, shapeName=dyn_front_hatstring_l_03_capsule, center=<2 0 0>, rotation=<rot 90 0 0>}]
 dyn_front_hatstring_l_04 [{shapeType=Capsule, length=2, radius=1, shapeName=dyn_front_hatstring_l_04_capsule, center=<2 0 0>, rotation=<rot 90 0 0>}]
 dyn_front_hatstring_l_05 [{shapeType=Capsule, length=2, radius=1, shapeName=dyn_front_hatstring_l_05_capsule, center=<2 0 0>, rotation=<rot 90 0 0>}]
 dyn_front_hatstring_l_06 [{shapeType=Capsule, length=1.4846, radius=1, shapeName=dyn_front_hatstring_l_06_capsule, center=<2 0 0>, rotation=<rot 90 0 0>}]
 dyn_front_hatstring_l_base [{shapeType=Capsule, length=0.505, radius=0.505, shapeName=dyn_front_hatstring_l_base_capsule, rotation=<rot 0 -90 0>}]
 dyn_front_hatstring_r_01 [{shapeType=Capsule, length=2.9406, radius=1, shapeName=dyn_front_hatstring_r_01_capsule, center=<2 0 0>, rotation=<rot 90 0 0>}]
 dyn_front_hatstring_r_02 [{shapeType=Capsule, length=4.5937, radius=1, shapeName=dyn_front_hatstring_r_02_capsule, center=<2 0 0>, rotation=<rot 90 0 0>}]
 dyn_front_hatstring_r_03 [{shapeType=Capsule, length=3.2028, radius=1, shapeName=dyn_front_hatstring_r_03_capsule, center=<2 0 0>, rotation=<rot 90 0 0>}]
 dyn_front_hatstring_r_04 [{shapeType=Capsule, length=2, radius=1, shapeName=dyn_front_hatstring_r_04_capsule, center=<2 0 0>, rotation=<rot 90 0 0>}]
 dyn_front_hatstring_r_05 [{shapeType=Capsule, length=2, radius=1, shapeName=dyn_front_hatstring_r_05_capsule, center=<2 0 0>, rotation=<rot 90 0 0>}]
 dyn_front_hatstring_r_06 [{shapeType=Capsule, length=1.4846, radius=1, shapeName=dyn_front_hatstring_r_06_capsule, center=<2 0 0>, rotation=<rot 90 0 0>}]
 dyn_front_hatstring_r_base [{shapeType=Capsule, length=0.505, radius=0.505, shapeName=dyn_front_hatstring_r_base_capsule, rotation=<rot 0 -90 0>}]
 dyn_hair_volume_l_a_01 [{shapeType=Capsule, length=4, radius=2, shapeName=dyn_hair_volume_l_a_01_capsule, center=<2 0 0>, rotation=<rot 90 90 0>}]
 dyn_hair_volume_l_a_02 [{shapeType=Capsule, length=4.356, radius=3.8985, shapeName=dyn_hair_volume_l_a_02_capsule, center=<2 0 0>, rotation=<rot 90 90 0>}]
 dyn_hair_volume_l_a_base [{shapeType=Capsule, length=0.505, radius=0.505, shapeName=dyn_hair_volume_l_a_base_capsule, rotation=<rot 0 -90 0>}]
 dyn_hair_volume_l_b_01 [{shapeType=Capsule, length=5.5485, radius=4.2246, shapeName=dyn_hair_volume_l_b_01_capsule, center=<2 0 -1>, rotation=<rot 0 0 90>}]
 dyn_hair_volume_l_b_02 [{shapeType=Capsule, length=4.7794, radius=4.0851, shapeName=dyn_hair_volume_l_b_02_capsule, center=<2 0 -1>, rotation=<rot 0 0 90>}]
 dyn_hair_volume_l_b_base [{shapeType=Capsule, length=0.505, radius=0.505, shapeName=dyn_hair_volume_l_b_base_capsule, rotation=<rot 0 -90 0>}]
 dyn_hair_volume_r_a_01 [{shapeType=Capsule, length=4, radius=2, shapeName=dyn_hair_volume_r_a_01_capsule, center=<-2 0 0>, rotation=<rot 90 90 0>}]
 dyn_hair_volume_r_a_02 [{shapeType=Capsule, length=4, radius=2, shapeName=dyn_hair_volume_r_a_02_capsule, center=<-2 0 0>, rotation=<rot 90 90 0>}]
 dyn_hair_volume_r_a_base [{shapeType=Capsule, length=0.505, radius=0.505, shapeName=dyn_hair_volume_r_a_base_capsule, rotation=<rot 0 -90 0>}]
 dyn_hair_volume_r_b_01 [{shapeType=Capsule, length=6.2389, radius=4.7739, shapeName=dyn_hair_volume_r_b_01_capsule, center=<-2 0 1>, rotation=<rot 0 0 90>}]
 dyn_hair_volume_r_b_02 [
  {shapeType=Capsule, length=6.9006, radius=2.2731, shapeName=dyn_hair_volume_r_b_02_capsule, center=<-2 0 1>, rotation=<rot 0 0 135.1326>}
  {
   shapeType=Capsule
   length=3.8143
   radius=3.4367
   shapeName=dyn_hair_volume_r_b_02_capsule
   center=<-2.2337 -1.9552 -0.2726>
   rotation=<rot 62.3126 -28.724 -59.9516>
  }
 ]
 dyn_hair_volume_r_b_base [{shapeType=Capsule, length=0.505, radius=0.505, shapeName=dyn_hair_volume_r_b_base_capsule, rotation=<rot 0 -90 0>}]
 dyn_hair_volume_rear_01 [{shapeType=Capsule, length=4.5582, radius=4.3367, shapeName=dyn_hair_volume_rear_01_capsule, center=<2 0 -1>, rotation=<rot 0 0 90>}]
 dyn_hair_volume_rear_02 [{shapeType=Capsule, length=5.5671, radius=3.8264, shapeName=dyn_hair_volume_rear_02_capsule, center=<2 0 -1>, rotation=<rot 0 0 90>}]
 dyn_hair_volume_rear_base [{shapeType=Capsule, length=0.505, radius=0.505, shapeName=dyn_hair_volume_rear_base_capsule, rotation=<rot 0 -90 0>}]
 dyn_hatstring_l_01 [{shapeType=Capsule, length=3, radius=1.5, shapeName=dyn_hatstring_l_01_capsule, center=<2 0 0>, rotation=<rot 90 90 0>}]
 dyn_hatstring_l_02 [{shapeType=Capsule, length=2.1149, radius=2.3851, shapeName=dyn_hatstring_l_02_capsule, center=<2 0 0>, rotation=<rot 90 90 0>}]
 dyn_hatstring_l_03 [{shapeType=Capsule, length=3, radius=1.5, shapeName=dyn_hatstring_l_03_capsule, center=<2 0 0>, rotation=<rot 90 90 0>}]
 dyn_hatstring_l_04 [{shapeType=Capsule, length=3, radius=1.5, shapeName=dyn_hatstring_l_04_capsule, center=<2 0 0>, rotation=<rot 90 90 0>}]
 dyn_hatstring_l_05 [{shapeType=Capsule, length=3, radius=1.5, shapeName=dyn_hatstring_l_05_capsule, center=<2 0 0>, rotation=<rot 90 90 0>}]
 dyn_hatstring_r_01 [{shapeType=Capsule, length=3, radius=1.5, shapeName=dyn_hatstring_r_01_capsule, center=<2 0 0.0706>, rotation=<rot 90 90 0>}]
 dyn_hatstring_r_02 [{shapeType=Capsule, length=2.1149, radius=2.3851, shapeName=dyn_hatstring_r_02_capsule, center=<2 0 0>, rotation=<rot 90 90 0>}]
 dyn_hatstring_r_03 [{shapeType=Capsule, length=3, radius=1.5, shapeName=dyn_hatstring_r_03_capsule, center=<2 0 0>, rotation=<rot 90 90 0>}]
 dyn_hatstring_r_04 [{shapeType=Capsule, length=3, radius=1.5, shapeName=dyn_hatstring_r_04_capsule, center=<2 0 0>, rotation=<rot 90 90 0>}]
 dyn_hatstring_r_05 [{shapeType=Capsule, length=3, radius=1.5, shapeName=dyn_hatstring_r_05_capsule, center=<2 0 0>, rotation=<rot 90 90 0>}]
 dyn_hatstring_r_06 [{shapeType=Capsule, length=3, radius=1.5, shapeName=dyn_hatstring_r_06_capsule, center=<2 0 0>, rotation=<rot 90 90 0>}]
 dyn_hatstrings_base [{
   shapeType=Capsule
   length=0.505
   radius=0.505
   shapeName=dyn_hatstrings_base_capsule
   center=<-0.6107 -0.2393 0.0173>
   rotation=<rot 0 -60.7772 -88.7146>
  }]
 dyn_rear_rotation_base [{shapeType=Capsule, length=0.505, radius=0.505, shapeName=dyn_rear_rotation_base_capsule, rotation=<rot 0 -90 0>}]
 head [{shapeType=Capsule, length=7.742, radius=7.2882, shapeName=head_capsule, center=<0.0096 1.4362 0>, rotation=<rot 90 2.4188 0>}]
 neck_01 [{shapeType=Capsule, length=5.6859, radius=5.3127, shapeName=neck_01_capsule, rotation=<rot 0 -90.0001 -89.8903>}]
 spine_05 [{shapeType=Capsule, length=13.243, radius=8.8321, shapeName=spine_05_capsule, center=<1.9605 -0.7432 0>, rotation=<rot 0 -90 0>}]
}
constraints [
 {
  bone1Name=neck_01
  bone2Name=spine_05
  swing1LimitDegrees=45
  swing1Motion=Free
  swing2LimitDegrees=45
  swing2Motion=Free
  twistLimitDegrees=45
  twistMotion=Free
 }
 {
  bone1Name=head
  bone2Name=neck_01
  swing1LimitDegrees=45
  swing1Motion=Free
  swing2LimitDegrees=45
  swing2Motion=Free
  twistLimitDegrees=45
  twistMotion=Free
 }
 {
  bone1Name=dyn_rear_rotation_base
  bone2Name=head
  swing1LimitDegrees=45
  swing1Motion=Free
  swing2LimitDegrees=45
  swing2Motion=Free
  twistLimitDegrees=45
  twistMotion=Free
 }
 {
  bone1Name=dyn_hair_volume_l_b_base
  bone2Name=dyn_rear_rotation_base
  swing1LimitDegrees=45
  swing1Motion=Locked
  swing2LimitDegrees=45
  swing2Motion=Locked
  twistLimitDegrees=45
  twistMotion=Locked
 }
 {bone1Name=dyn_hair_volume_l_b_01, bone2Name=dyn_hair_volume_l_b_base, swing1LimitDegrees=8, swing2LimitDegrees=8, twistLimitDegrees=8}
 {bone1Name=dyn_hair_volume_l_b_02, bone2Name=dyn_hair_volume_l_b_01, swing1LimitDegrees=24, swing2LimitDegrees=24, twistLimitDegrees=24}
 {
  bone1Name=dyn_hair_volume_r_b_base
  bone2Name=dyn_rear_rotation_base
  swing1LimitDegrees=45
  swing1Motion=Locked
  swing2LimitDegrees=45
  swing2Motion=Locked
  twistLimitDegrees=45
  twistMotion=Locked
 }
 {bone1Name=dyn_hair_volume_r_b_01, bone2Name=dyn_hair_volume_r_b_base, swing1LimitDegrees=8, swing2LimitDegrees=8, twistLimitDegrees=8}
 {bone1Name=dyn_hair_volume_r_b_02, bone2Name=dyn_hair_volume_r_b_01, swing1LimitDegrees=24, swing2LimitDegrees=24, twistLimitDegrees=24}
 {
  bone1Name=dyn_hair_volume_rear_base
  bone2Name=dyn_rear_rotation_base
  swing1LimitDegrees=45
  swing1Motion=Free
  swing2LimitDegrees=45
  swing2Motion=Free
  twistLimitDegrees=45
  twistMotion=Free
 }
 {bone1Name=dyn_hair_volume_rear_01, bone2Name=dyn_hair_volume_rear_base, swing1LimitDegrees=24, swing2LimitDegrees=8, twistLimitDegrees=24}
 {bone1Name=dyn_hair_volume_rear_02, bone2Name=dyn_hair_volume_rear_01, swing1LimitDegrees=24, swing2LimitDegrees=24, twistLimitDegrees=24}
 {
  bone1Name=dyn_hair_volume_l_a_base
  bone2Name=head
  swing1LimitDegrees=45
  swing1Motion=Free
  swing2LimitDegrees=45
  swing2Motion=Free
  twistLimitDegrees=45
  twistMotion=Free
 }
 {bone1Name=dyn_hair_volume_l_a_01, bone2Name=dyn_hair_volume_l_a_base, swing1LimitDegrees=12, swing2LimitDegrees=12, twistLimitDegrees=12}
 {bone1Name=dyn_hair_volume_l_a_02, bone2Name=dyn_hair_volume_l_a_01, swing1LimitDegrees=12, swing2LimitDegrees=12, twistLimitDegrees=12}
 {
  bone1Name=dyn_hair_volume_r_a_base
  bone2Name=head
  swing1LimitDegrees=45
  swing1Motion=Free
  swing2LimitDegrees=45
  swing2Motion=Free
  twistLimitDegrees=45
  twistMotion=Free
 }
 {bone1Name=dyn_hair_volume_r_a_01, bone2Name=dyn_hair_volume_r_a_base, swing1LimitDegrees=8, swing2LimitDegrees=8, twistLimitDegrees=8}
 {bone1Name=dyn_hair_volume_r_a_02, bone2Name=dyn_hair_volume_r_a_01, swing1LimitDegrees=12, swing2LimitDegrees=12, twistLimitDegrees=12}
 {
  bone1Name=dyn_hatstrings_base
  bone2Name=deform_hat
  swing1LimitDegrees=45
  swing1Motion=Locked
  swing2LimitDegrees=45
  swing2Motion=Locked
  twistLimitDegrees=45
  twistMotion=Locked
 }
 {bone1Name=dyn_hatstring_l_01, bone2Name=dyn_hatstrings_base, swing1LimitDegrees=4, swing2LimitDegrees=4, twistLimitDegrees=4}
 {bone1Name=dyn_hatstring_l_02, bone2Name=dyn_hatstring_l_01, swing1LimitDegrees=4, swing2LimitDegrees=4, twistLimitDegrees=4}
 {bone1Name=dyn_hatstring_l_03, bone2Name=dyn_hatstring_l_02, swing1LimitDegrees=20, swing2LimitDegrees=20, twistLimitDegrees=20}
 {bone1Name=dyn_hatstring_l_04, bone2Name=dyn_hatstring_l_03, swing1LimitDegrees=20, swing2LimitDegrees=20, twistLimitDegrees=20}
 {bone1Name=dyn_hatstring_l_05, bone2Name=dyn_hatstring_l_04, swing1LimitDegrees=20, swing2LimitDegrees=20, twistLimitDegrees=20}
 {bone1Name=dyn_hatstring_r_01, bone2Name=dyn_hatstrings_base, swing1LimitDegrees=10, swing2LimitDegrees=10, twistLimitDegrees=10}
 {bone1Name=dyn_hatstring_r_02, bone2Name=dyn_hatstring_r_01, swing1LimitDegrees=8, swing2LimitDegrees=20, twistLimitDegrees=20}
 {bone1Name=dyn_hatstring_r_03, bone2Name=dyn_hatstring_r_02, swing1LimitDegrees=20, swing2LimitDegrees=20, twistLimitDegrees=20}
 {bone1Name=dyn_hatstring_r_04, bone2Name=dyn_hatstring_r_03, swing1LimitDegrees=20, swing2LimitDegrees=20, twistLimitDegrees=20}
 {bone1Name=dyn_hatstring_r_05, bone2Name=dyn_hatstring_r_04, swing1LimitDegrees=20, swing2LimitDegrees=20, twistLimitDegrees=20}
 {bone1Name=dyn_hatstring_r_06, bone2Name=dyn_hatstring_r_05, swing1LimitDegrees=20, swing2LimitDegrees=20, twistLimitDegrees=20}
 {
  bone1Name=dyn_hatstring_l_02
  bone2Name=dyn_hat_lip_rear_base
  swing1LimitDegrees=45
  swing1Motion=Free
  swing2LimitDegrees=45
  swing2Motion=Free
  twistLimitDegrees=45
  twistMotion=Free
 }
 {
  bone1Name=dyn_hatstring_r_02
  bone2Name=dyn_hat_lip_rear_base
  swing1LimitDegrees=45
  swing1Motion=Free
  swing2LimitDegrees=45
  swing2Motion=Free
  twistLimitDegrees=45
  twistMotion=Free
 }
 {
  bone1Name=deform_hat
  bone2Name=head
  swing1LimitDegrees=45
  swing1Motion=Free
  swing2LimitDegrees=45
  swing2Motion=Free
  twistLimitDegrees=45
  twistMotion=Free
 }
 {
  bone1Name=dyn_front_hatstring_l_base
  bone2Name=deform_hat
  swing1LimitDegrees=45
  swing1Motion=Free
  swing2LimitDegrees=45
  swing2Motion=Free
  twistLimitDegrees=45
  twistMotion=Free
 }
 {bone1Name=dyn_front_hatstring_l_01, bone2Name=dyn_front_hatstring_l_base, swing1LimitDegrees=2, swing2LimitDegrees=2, twistLimitDegrees=2}
 {bone1Name=dyn_front_hatstring_l_02, bone2Name=dyn_front_hatstring_l_01, swing1LimitDegrees=4, swing2LimitDegrees=4, twistLimitDegrees=4}
 {bone1Name=dyn_front_hatstring_l_03, bone2Name=dyn_front_hatstring_l_02, swing1LimitDegrees=8, swing2LimitDegrees=8, twistLimitDegrees=8}
 {bone1Name=dyn_front_hatstring_l_04, bone2Name=dyn_front_hatstring_l_03, swing1LimitDegrees=8, swing2LimitDegrees=8, twistLimitDegrees=8}
 {bone1Name=dyn_front_hatstring_l_05, bone2Name=dyn_front_hatstring_l_04, swing1LimitDegrees=8, swing2LimitDegrees=8, twistLimitDegrees=8}
 {bone1Name=dyn_front_hatstring_l_06, bone2Name=dyn_front_hatstring_l_05, swing1LimitDegrees=8, swing2LimitDegrees=8, twistLimitDegrees=8}
 {
  bone1Name=dyn_front_hatstring_r_base
  bone2Name=deform_hat
  swing1LimitDegrees=45
  swing1Motion=Free
  swing2LimitDegrees=45
  swing2Motion=Free
  twistLimitDegrees=45
  twistMotion=Free
 }
 {bone1Name=dyn_front_hatstring_r_01, bone2Name=dyn_front_hatstring_r_base, swing1LimitDegrees=2, swing2LimitDegrees=2, twistLimitDegrees=2}
 {bone1Name=dyn_front_hatstring_r_02, bone2Name=dyn_front_hatstring_r_01, swing1LimitDegrees=8, swing2LimitDegrees=8, twistLimitDegrees=8}
 {bone1Name=dyn_front_hatstring_r_03, bone2Name=dyn_front_hatstring_r_02, swing1LimitDegrees=8, swing2LimitDegrees=8, twistLimitDegrees=8}
 {bone1Name=dyn_front_hatstring_r_04, bone2Name=dyn_front_hatstring_r_03, swing1LimitDegrees=8, swing2LimitDegrees=8, twistLimitDegrees=8}
 {bone1Name=dyn_front_hatstring_r_05, bone2Name=dyn_front_hatstring_r_04, swing1LimitDegrees=8, swing2LimitDegrees=8, twistLimitDegrees=8}
 {bone1Name=dyn_front_hatstring_r_06, bone2Name=dyn_front_hatstring_r_05, swing1LimitDegrees=8, swing2LimitDegrees=8, twistLimitDegrees=8}
 {
  bone1Name=dyn_front_hatstring_l_06
  bone2Name=dyn_front_hatstring_r_06
  swing1LimitDegrees=45
  swing1Motion=Locked
  swing2LimitDegrees=45
  swing2Motion=Locked
  twistLimitDegrees=45
  twistMotion=Locked
 }
 {
  bone1Name=dyn_front_hatstring_r_06
  bone2Name=dyn_front_hatstring_l_06
  swing1LimitDegrees=45
  swing1Motion=Locked
  swing2LimitDegrees=45
  swing2Motion=Locked
  twistLimitDegrees=45
  twistMotion=Locked
 }
 {
  bone1Name=dyn_collar_l
  bone2Name=spine_05
  swing1LimitDegrees=45
  swing1Motion=Free
  swing2LimitDegrees=45
  swing2Motion=Free
  twistLimitDegrees=45
  twistMotion=Free
 }
 {
  bone1Name=dyn_collar_r
  bone2Name=spine_05
  swing1LimitDegrees=45
  swing1Motion=Free
  swing2LimitDegrees=45
  swing2Motion=Free
  twistLimitDegrees=45
  twistMotion=Free
 }
]
