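# Animation Blueprint for the HomeRange female outfit's face accessory mesh, posing and simulating its hair, hat and collar.
# Copied parent and body poses (blended at collar bones) feed a Control Rig, RigidBody hair physics, corrective bone offsets, then AnimDynamics for fringe and hat brim.
# The event graph computes gravity override from falling state, sets front-end flags and hat sim settings, and changes Chaos cloth gravity around skydiving and falling events.
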AnimBlueprint /BRCosmetics/Characters/Player/Female/Medium/Bodies/F_MED_HomeRange/Meshes/Parts/F_MED_HomeRange_FaceAcc_AnimBP
target_skeleton=&/F_MED_HomeRange_FaceAcc_Skeleton
graphs [
 AnimGraph {nodes [
   n1 = Get Body SkelMesh() -> (Body SkelMesh:Skeletal Mesh Component Budgeted Object Reference)
   n2 = Control Rig!(
    Source=Layered blend per bone(
     BasePose=Copy Pose From Mesh() {bUseAttachedParent=true}
     BlendPoses_0=Copy Pose From Mesh(SourceMeshComponent=@n1) {bUseAttachedParent=true}
    ) {layerSetup [{branchFilters [{boneName=dyn_collar_l}, {boneName=dyn_collar_r}]}]}
    Alpha
    isOnGround:double="To Float (Boolean)"(InBool=Get isOnGround() -> (isOnGround:bool))
   ) -> (Pose) {
    alphaBoolBlend {blendOption=Linear}
    controlRigAssetReference {blueprintRigClass=&/F_MED_HomeRange_FaceAcc_CtrlRig.F_MED_HomeRange_FaceAcc_CtrlRig_C}
   }
   n3 = RigidBody!(ComponentPose=Local To Component(LocalPose=@n2), Alpha, OverrideWorldGravity:Vector="0,0,-980", ExternalForce) -> (Pose) {
    bOverrideWorldGravity=true
    overridePhysicsAsset=&/F_MED_HomeRange_FaceAcc_Hair_Physics
    simulationSpace=BaseBoneSpace
    alphaBoolBlend {blendOption=Linear}
    baseBoneRef {boneName=head}
    componentAppliedLinearAccClamp=<0 0 0>
    overrideWorldGravity {z=-980}
    simSpaceSettings {
     dampingAlpha=0
     maxAngularAcceleration=6
     maxAngularVelocity=100
     maxLinearAcceleration=80
     maxLinearVelocity=1400
     velocityScaleZ=0.8
     worldAlpha=1
     externalLinearDragV {y=0.2, z=0.45}
    }
   }
   n4 = "Transform (Modify) Bone - Bone: dyn_hair_volume_rear_base"!(ComponentPose=@n3, AlphaCurveName:Name=SPR Head Rot Back, Translation="0,0,0", Rotation="0,0,0", Scale="1,1,0.5") -> (Pose) {
    alphaInputType=Curve
    scaleMode=BMM_Additive
    scaleSpace=BCS_BoneSpace
    alphaBoolBlend {blendOption=Linear}
    boneToModify {boneName=dyn_hair_volume_rear_base}
    scale {z=0.5}
   }
   n5 = "Transform (Modify) Bone - Bone: dyn_hair_volume_r_a_base"!(ComponentPose=@n4, AlphaCurveName:Name=SPR Head Rot Back, Translation="0,0,0", Rotation="0,0,0", Scale="1.000,0.643,1.000") -> (Pose) {
    alphaInputType=Curve
    scaleMode=BMM_Additive
    scaleSpace=BCS_BoneSpace
    alphaBoolBlend {blendOption=Linear}
    boneToModify {boneName=dyn_hair_volume_r_a_base}
    scale {y=0.643}
   }
   n6 = "Transform (Modify) Bone - Bone: dyn_hair_volume_rear_base"!(ComponentPose=@n5, bAlphaBoolEnabled:bool=True, Translation="0,0,0", Rotation="0,0,0", Scale="1,1,0.5") -> (Pose) {
    alphaInputType=Bool
    scaleMode=BMM_Additive
    scaleSpace=BCS_BoneSpace
    alphaBoolBlend {blendOption=Linear}
    boneToModify {boneName=dyn_hair_volume_rear_base}
    scale {z=0.5}
   }
   n7 = "Transform (Modify) Bone - Bone: dyn_hair_volume_r_a_base"!(ComponentPose=@n6, bAlphaBoolEnabled:bool=True, Translation="0,0,0", Rotation="0,0,0", Scale="1.000,0.643,1.000") -> (Pose) {
    alphaInputType=Bool
    scaleMode=BMM_Additive
    scaleSpace=BCS_BoneSpace
    alphaBoolBlend {blendOption=Linear}
    boneToModify {boneName=dyn_hair_volume_r_a_base}
    scale {y=0.643}
   }
   n8 = "Transform (Modify) Bone - Bone: dyn_hair_volume_l_b_base"!(ComponentPose=@n7, bAlphaBoolEnabled:bool=True, Translation="0,0,0", Rotation="0,-10,-10", Scale="1,0.6,0.8") -> (Pose) {
    alphaInputType=Bool
    rotationMode=BMM_Additive
    rotationSpace=BCS_BoneSpace
    scaleMode=BMM_Additive
    scaleSpace=BCS_BoneSpace
    alphaBoolBlend {blendOption=Linear}
    boneToModify {boneName=dyn_hair_volume_l_b_base}
    rotation {roll=-10, yaw=-10}
    scale {y=0.6, z=0.8}
   }
   n9 = "Transform (Modify) Bone - Bone: dyn_hat_lip_rear_01"!(ComponentPose=@n8, bAlphaBoolEnabled:bool=True, Translation="0,0,0", Rotation="-10,0,0", Scale="1,1,1") -> (Pose) {
    alphaInputType=Bool
    rotationMode=BMM_Additive
    rotationSpace=BCS_BoneSpace
    scaleMode=BMM_Additive
    scaleSpace=BCS_BoneSpace
    alphaBoolBlend {blendOption=Linear}
    boneToModify {boneName=dyn_hat_lip_rear_01}
    rotation {pitch=-10}
   }
   n10 = Anim Dynamics - Bone: dyn_hair_fringe_a_01(ComponentPose=@n9, Alpha=1, ExternalForce="0,0,0") {
    angularDampingOverride=1
    angularSpringConstant=20000
    bAngularSpring=true
    bOverrideAngularDamping=true
    bUseGravityOverride=true
    gravityScale=0.35
    simulationSpace=RootRelative
    alphaBoolBlend {blendOption=Linear}
    boundBone {boneName=dyn_hair_fringe_a_01}
    componentLinearAccScale=<1 1 1>
    componentLinearVelScale=<1 1 1>
    gravityOverride {z=-980}
    physicsBodyDefinitions [{
      boundBone {boneName=dyn_hair_fringe_a_01}
      boxExtents=<20 30 30>
      constraintSetup {angularLimitsMax=<5 5 5>, angularLimitsMin=<-5 -15 -5>, angularTarget=<1 0 0>}
      localJointOffset=<4 0 0>
     }]
   }
   n11 = Anim Dynamics - Bone: dyn_hair_fringe_a_02(ComponentPose=@n10, Alpha=1, ExternalForce="0,0,0") {
    angularDampingOverride=1
    angularSpringConstant=20000
    bAngularSpring=true
    bOverrideAngularDamping=true
    bUseGravityOverride=true
    gravityScale=0.35
    simulationSpace=RootRelative
    alphaBoolBlend {blendOption=Linear}
    boundBone {boneName=dyn_hair_fringe_a_02}
    componentLinearAccScale=<1 1 1>
    componentLinearVelScale=<1 1 1>
    gravityOverride {z=-980}
    physicsBodyDefinitions [{
      boundBone {boneName=dyn_hair_fringe_a_02}
      boxExtents=<20 30 30>
      constraintSetup {angularLimitsMax=<10 10 10>, angularLimitsMin=<-10 -15 -10>, angularTarget=<1 0 0>}
      localJointOffset=<4 0 0>
     }]
   }
   n12 = Anim Dynamics - Bone: dyn_hair_fringe_b_01(ComponentPose=@n11, Alpha=1, ExternalForce="0,0,0") {
    angularDampingOverride=1
    angularSpringConstant=20000
    bAngularSpring=true
    bOverrideAngularDamping=true
    bUseGravityOverride=true
    gravityScale=0.35
    simulationSpace=RootRelative
    alphaBoolBlend {blendOption=Linear}
    boundBone {boneName=dyn_hair_fringe_b_01}
    componentLinearAccScale=<1 1 1>
    componentLinearVelScale=<1 1 1>
    gravityOverride {z=-980}
    physicsBodyDefinitions [{
      boundBone {boneName=dyn_hair_fringe_b_01}
      boxExtents=<20 30 30>
      constraintSetup {angularLimitsMax=<5 5 5>, angularLimitsMin=<-5 -15 -5>, angularTarget=<1 0 0>}
      localJointOffset=<4 0 0>
     }]
   }
   n13 = Anim Dynamics - Bone: dyn_hair_fringe_b_02(ComponentPose=@n12, Alpha=1, ExternalForce="0,0,0") {
    angularDampingOverride=1
    angularSpringConstant=20000
    bAngularSpring=true
    bOverrideAngularDamping=true
    bUseGravityOverride=true
    gravityScale=0.35
    simulationSpace=RootRelative
    alphaBoolBlend {blendOption=Linear}
    boundBone {boneName=dyn_hair_fringe_b_02}
    componentLinearAccScale=<1 1 1>
    componentLinearVelScale=<1 1 1>
    gravityOverride {z=-980}
    physicsBodyDefinitions [{
      boundBone {boneName=dyn_hair_fringe_b_02}
      boxExtents=<20 30 30>
      constraintSetup {angularLimitsMax=<10 10 10>, angularLimitsMin=<-10 -15 -10>, angularTarget=<1 0 0>}
      localJointOffset=<4 0 0>
     }]
   }
   n14 = Anim Dynamics - Bone: dyn_hair_fringe_b_03(ComponentPose=@n13, Alpha=1, ExternalForce="0,0,0") {
    angularDampingOverride=1
    angularSpringConstant=20000
    bAngularSpring=true
    bOverrideAngularDamping=true
    bUseGravityOverride=true
    gravityScale=0.35
    simulationSpace=RootRelative
    alphaBoolBlend {blendOption=Linear}
    boundBone {boneName=dyn_hair_fringe_b_03}
    componentLinearAccScale=<1 1 1>
    componentLinearVelScale=<1 1 1>
    gravityOverride {z=-980}
    physicsBodyDefinitions [{
      boundBone {boneName=dyn_hair_fringe_b_03}
      boxExtents=<20 30 30>
      constraintSetup {angularLimitsMax=<15 15 15>, angularLimitsMin=<-15 -15 -15>, angularTarget=<1 0 0>}
      localJointOffset=<4 0 0>
     }]
   }
   n15 = Anim Dynamics - Bone: dyn_hair_fringe_b_04(ComponentPose=@n14, Alpha=1, ExternalForce="0,0,0") {
    angularDampingOverride=1
    angularSpringConstant=20000
    bAngularSpring=true
    bOverrideAngularDamping=true
    bUseGravityOverride=true
    gravityScale=0.35
    simulationSpace=RootRelative
    alphaBoolBlend {blendOption=Linear}
    boundBone {boneName=dyn_hair_fringe_b_04}
    componentLinearAccScale=<1 1 1>
    componentLinearVelScale=<1 1 1>
    gravityOverride {z=-980}
    physicsBodyDefinitions [{
      boundBone {boneName=dyn_hair_fringe_b_04}
      boxExtents=<20 30 30>
      constraintSetup {angularLimitsMax=<15 15 15>, angularLimitsMin=<-15 -15 -15>, angularTarget=<1 0 0>}
      localJointOffset=<4 0 0>
     }]
   }
   n16 = Anim Dynamics - Bone: dyn_hair_fringe_c_01(ComponentPose=@n15, Alpha=1, ExternalForce="0,0,0") {
    angularDampingOverride=1
    angularSpringConstant=20000
    bAngularSpring=true
    bOverrideAngularDamping=true
    bUseGravityOverride=true
    gravityScale=0.35
    simulationSpace=RootRelative
    alphaBoolBlend {blendOption=Linear}
    boundBone {boneName=dyn_hair_fringe_c_01}
    componentLinearAccScale=<1 1 1>
    componentLinearVelScale=<1 1 1>
    gravityOverride {z=-980}
    physicsBodyDefinitions [{
      boundBone {boneName=dyn_hair_fringe_c_01}
      boxExtents=<20 30 30>
      constraintSetup {angularLimitsMax=<5 5 5>, angularLimitsMin=<-5 0 -5>, angularTarget=<1 0 0>}
      localJointOffset=<4 0 0>
     }]
   }
   n17 = Anim Dynamics - Bone: dyn_hair_fringe_c_02(ComponentPose=@n16, Alpha=1, ExternalForce="0,0,0") {
    angularDampingOverride=1
    angularSpringConstant=20000
    bAngularSpring=true
    bOverrideAngularDamping=true
    bUseGravityOverride=true
    gravityScale=0.35
    simulationSpace=RootRelative
    alphaBoolBlend {blendOption=Linear}
    boundBone {boneName=dyn_hair_fringe_c_02}
    componentLinearAccScale=<1 1 1>
    componentLinearVelScale=<1 1 1>
    gravityOverride {z=-980}
    physicsBodyDefinitions [{
      boundBone {boneName=dyn_hair_fringe_c_02}
      boxExtents=<20 30 30>
      constraintSetup {angularLimitsMax=<10 10 10>, angularLimitsMin=<-10 0 -10>, angularTarget=<1 0 0>}
      localJointOffset=<4 0 0>
     }]
   }
   n18 = Anim Dynamics - Bone: dyn_hair_fringe_c_03(ComponentPose=@n17, Alpha=1, ExternalForce="0,0,0") {
    angularDampingOverride=1
    angularSpringConstant=20000
    bAngularSpring=true
    bOverrideAngularDamping=true
    bUseGravityOverride=true
    gravityScale=0.35
    simulationSpace=RootRelative
    alphaBoolBlend {blendOption=Linear}
    boundBone {boneName=dyn_hair_fringe_c_03}
    componentLinearAccScale=<1 1 1>
    componentLinearVelScale=<1 1 1>
    gravityOverride {z=-980}
    physicsBodyDefinitions [{
      boundBone {boneName=dyn_hair_fringe_c_03}
      boxExtents=<20 30 30>
      constraintSetup {angularLimitsMax=<15 15 15>, angularLimitsMin=<-15 0 -15>, angularTarget=<1 0 0>}
      localJointOffset=<4 0 0>
     }]
   }
   n19 = Anim Dynamics - Bone: dyn_hat_lip_l_01!(
    ComponentPose=@n18
    Alpha=1
    SimSpaceSettings:Anim Phys Sim Space Settings Structure=(SimSpaceAngularAlpha=0,MaxAngularVelocity=10000,MaxAngularAcceleration=10000,ExternalAngularVelocity=(X=0,Y=0,Z=0))
    ExternalForce="0,0,0"
   ) -> (Pose) {
    angularDampingOverride=1
    angularSpringConstant=25000
    bAngularSpring=true
    bOverrideAngularDamping=true
    bUseGravityOverride=true
    gravityScale=0.35
    simulationSpace=BoneRelative
    alphaBoolBlend {blendOption=Linear}
    boundBone {boneName=dyn_hat_lip_l_01}
    componentLinearVelScale {y=1}
    gravityOverride {z=-980}
    physicsBodyDefinitions [{
      boundBone {boneName=dyn_hat_lip_l_01}
      boxExtents=<30 30 30>
      constraintSetup {angularLimitsMax=<0 10 0>, angularLimitsMin=<0 -10 0>, angularTarget=<1 0 0>}
      localJointOffset=<2 0 0>
     }]
    relativeSpaceBone {boneName=head}
    simSpaceSettings {maxAngularAcceleration=30, maxAngularVelocity=300, simSpaceAngularAlpha=1}
   }
   n20 = Anim Dynamics - Bone: dyn_hat_lip_l_02!(
    ComponentPose=@n19
    Alpha=1
    SimSpaceSettings:Anim Phys Sim Space Settings Structure=(SimSpaceAngularAlpha=1,MaxAngularVelocity=300,MaxAngularAcceleration=30,ExternalAngularVelocity=(X=0,Y=0,Z=0))
    ExternalForce="0,0,0"
   ) -> (Pose) {
    angularDampingOverride=1
    angularSpringConstant=25000
    bAngularSpring=true
    bOverrideAngularDamping=true
    bUseGravityOverride=true
    gravityScale=0.35
    simulationSpace=BoneRelative
    alphaBoolBlend {blendOption=Linear}
    boundBone {boneName=dyn_hat_lip_l_02}
    componentLinearVelScale {y=1}
    gravityOverride {z=-980}
    physicsBodyDefinitions [{
      boundBone {boneName=dyn_hat_lip_l_02}
      boxExtents=<30 30 30>
      constraintSetup {angularLimitsMax=<0 10 0>, angularLimitsMin=<0 -10 0>, angularTarget=<1 0 0>}
      localJointOffset=<2 0 0>
     }]
    relativeSpaceBone {boneName=head}
    simSpaceSettings {maxAngularAcceleration=30, maxAngularVelocity=300, simSpaceAngularAlpha=1}
   }
   n21 = Anim Dynamics - Bone: dyn_hat_lip_r_01!(
    ComponentPose=@n20
    Alpha=1
    SimSpaceSettings:Anim Phys Sim Space Settings Structure=(SimSpaceAngularAlpha=1,MaxAngularVelocity=300,MaxAngularAcceleration=30,ExternalAngularVelocity=(X=0,Y=0,Z=0))
    ExternalForce="0,0,0"
   ) -> (Pose) {
    angularDampingOverride=1
    angularSpringConstant=25000
    bAngularSpring=true
    bOverrideAngularDamping=true
    bUseGravityOverride=true
    gravityScale=0.35
    simulationSpace=BoneRelative
    alphaBoolBlend {blendOption=Linear}
    boundBone {boneName=dyn_hat_lip_r_01}
    componentLinearVelScale {y=1}
    gravityOverride {z=-980}
    physicsBodyDefinitions [{
      boundBone {boneName=dyn_hat_lip_r_01}
      boxExtents=<30 30 30>
      constraintSetup {angularLimitsMax=<0 10 0>, angularLimitsMin=<0 -10 0>, angularTarget=<1 0 0>}
      localJointOffset=<-2 0 0>
     }]
    relativeSpaceBone {boneName=head}
    simSpaceSettings {maxAngularAcceleration=30, maxAngularVelocity=300, simSpaceAngularAlpha=1}
   }
   n22 = Anim Dynamics - Bone: dyn_hat_lip_r_02!(
    ComponentPose=@n21
    Alpha=1
    SimSpaceSettings:Anim Phys Sim Space Settings Structure=(SimSpaceAngularAlpha=1,MaxAngularVelocity=300,MaxAngularAcceleration=30,ExternalAngularVelocity=(X=0,Y=0,Z=0))
    ExternalForce="0,0,0"
   ) -> (Pose) {
    angularDampingOverride=1
    angularSpringConstant=25000
    bAngularSpring=true
    bOverrideAngularDamping=true
    bUseGravityOverride=true
    gravityScale=0.35
    simulationSpace=BoneRelative
    alphaBoolBlend {blendOption=Linear}
    boundBone {boneName=dyn_hat_lip_r_02}
    componentLinearVelScale {y=1}
    gravityOverride {z=-980}
    physicsBodyDefinitions [{
      boundBone {boneName=dyn_hat_lip_r_02}
      boxExtents=<30 30 30>
      constraintSetup {angularLimitsMax=<0 10 0>, angularLimitsMin=<0 -10 0>, angularTarget=<1 0 0>}
      localJointOffset=<-2 0 0>
     }]
    relativeSpaceBone {boneName=head}
    simSpaceSettings {maxAngularAcceleration=30, maxAngularVelocity=300, simSpaceAngularAlpha=1}
   }
   n23 = "Transform (Modify) Bone - Bone: dyn_front_hatstring_l_01"(ComponentPose=@n22, Alpha=1, Translation="0,0,-0.5", Rotation="0,0,0", Scale="1,1,1") {
    rotationMode=BMM_Additive
    rotationSpace=BCS_BoneSpace
    translationMode=BMM_Additive
    translationSpace=BCS_BoneSpace
    alphaBoolBlend {blendOption=Linear}
    boneToModify {boneName=dyn_front_hatstring_l_01}
    translation {z=-0.5}
   }
   n24 = "Transform (Modify) Bone - Bone: dyn_front_hatstring_l_03"(ComponentPose=@n23, Alpha=1, Translation="0,0,0.5", Rotation="0,0,0", Scale="1,1,1") {
    rotationMode=BMM_Additive
    rotationSpace=BCS_BoneSpace
    translationMode=BMM_Additive
    translationSpace=BCS_BoneSpace
    alphaBoolBlend {blendOption=Linear}
    boneToModify {boneName=dyn_front_hatstring_l_03}
    translation {z=0.5}
   }
   n25 = "Transform (Modify) Bone - Bone: dyn_hair_volume_r_b_01"!(ComponentPose=@n24, bAlphaBoolEnabled:bool=True, Translation="0,0,0", Rotation="-12.978,-0.204,-13.920", Scale="1,1,1") -> (Pose) {
    alphaInputType=Bool
    rotationMode=BMM_Additive
    rotationSpace=BCS_BoneSpace
    translationMode=BMM_Additive
    translationSpace=BCS_BoneSpace
    alphaBoolBlend {blendOption=Linear}
    boneToModify {boneName=dyn_hair_volume_r_b_01}
    rotation=<rot -12.978 -0.204 -13.92>
   }
   n26 = "Transform (Modify) Bone - Bone: dyn_hair_volume_r_a_01"!(ComponentPose=@n25, bAlphaBoolEnabled:bool=True, Translation="0,0,0", Rotation="-0.010,-0.026,-13.875", Scale="1,1,1") -> (Pose) {
    alphaInputType=Bool
    rotationMode=BMM_Additive
    rotationSpace=BCS_BoneSpace
    translationMode=BMM_Additive
    translationSpace=BCS_BoneSpace
    alphaBoolBlend {blendOption=Linear}
    boneToModify {boneName=dyn_hair_volume_r_a_01}
    rotation=<rot -0.01 -0.026 -13.875>
   }
   n27 = "Transform (Modify) Bone - Bone: dyn_front_hatstring_l_base"(ComponentPose=@n26, Alpha=1, Translation="0,0,0", Rotation="-5,-15,0", Scale="1,1,1") {
    rotationMode=BMM_Additive
    rotationSpace=BCS_BoneSpace
    translationMode=BMM_Additive
    translationSpace=BCS_BoneSpace
    alphaBoolBlend {blendOption=Linear}
    boneToModify {boneName=dyn_front_hatstring_l_base}
    rotation {pitch=-5, yaw=-15}
   }
   n28 = "Transform (Modify) Bone - Bone: dyn_front_hatstring_r_base"(ComponentPose=@n27, Alpha=1, Translation="-1.5,0,0", Rotation="-5,5,0", Scale="1,1,1") {
    rotationMode=BMM_Additive
    rotationSpace=BCS_BoneSpace
    translationMode=BMM_Additive
    translationSpace=BCS_BoneSpace
    alphaBoolBlend {blendOption=Linear}
    boneToModify {boneName=dyn_front_hatstring_r_base}
    rotation {pitch=-5, yaw=5}
    translation {x=-1.5}
   }
   n29 = "Transform (Modify) Bone - Bone: dyn_hat_lip_rear_base"(ComponentPose=@n28, Alpha=1, Translation="0,0,0", Rotation="-15,0,0", Scale="1,1,1") {
    rotationMode=BMM_Additive
    rotationSpace=BCS_BoneSpace
    translationMode=BMM_Additive
    translationSpace=BCS_BoneSpace
    alphaBoolBlend {blendOption=Linear}
    boneToModify {boneName=dyn_hat_lip_rear_base}
    rotation {pitch=-15}
   }
   Output Pose(Result=Component To Local(ComponentPose=@n29)) {layerGroup=None}
  ]}
 EventGraph {nodes [
   n1 = Event BlueprintUpdateAnimation()
   > n2 = Sequence()
   > n3 = "Gravity Override State Function (Parameter Asset) (Nativized)"(
    CharacterPartAnimInstance=Self-Reference()
    GameTimeInSeconds=GetGameTimeInSeconds()
    Parameters=&/F_MED_HomeRange_Params
   )
   > Set Gravity Override(Gravity Override=@n3)
   n4 = OR Boolean(A=Get bIsSkyDiving(), B=Get bIsSkyDivingFromBus() -> (bIsSkyDivingFromBus:bool))
   > Branch(Condition=AND Boolean(A=Get bIsFalling(), B=NOT Boolean(A=@n4)))
   n5 = float * float(A=@n1.DeltaTimeX, B=40) {is=title K2Node_PromotableOperator}
   > Set Gravity Override(Gravity Override="Lerp (Vector)"!(A:Vector=Get Gravity Override(), B:Vector="0,0,400", Alpha:float="Clamp (Float)"(Value=@n5, Min=0, Max=1)) -> (ReturnValue:Vector))
   n6 = Get bRigidBodyEnabled()
   n7 = Get bIsPlayingEmote()
   Set FrontEnd Low Setting / IOS devive(
    execute=@n2.then_1
    FrontEnd Low Setting / IOS devive:bool=AND Boolean!(A=Get bIsInFrontEnd(), B=NOT Boolean(A=@n6), C:bool=NOT Boolean(A=@n7)) -> (ReturnValue)
   ) -> (Output_Get:bool)
   > Set isOnGround(isOnGround:bool=NOT Boolean(A=Get bIsSkyDiving())) -> (Output_Get:bool)
   > Branch(Condition=Get bIsInFrontEnd())
   > Set Hat Sim Space Settings(Hat Sim Space Settings:Anim Phys Sim Space Settings Structure=Get Hat Sim Space Settings Frontend() -> (Hat Sim Space Settings Frontend:Anim Phys Sim Space Settings Structure)) -> (Output_Get:Anim Phys Sim Space Settings Structure)
   n8 = Event OnBeginSkydiving()
   SetClothingInteractorIfNotSet(execute=[@n8.then, @n12.then])
   > SetGravity!(
    execute
    self:Chaos Clothing Interactor Object Reference=Get ChaosClothingInteractor() -> (ChaosClothingInteractor:Chaos Clothing Interactor Object Reference)
    GravityScale:float=1
    bIsGravityOverridden:bool=false
    GravityOverride:Vector="0,0,-2000"
   ) -> (then)
   n9 = Event OnCharacterCustomizationLoaded()
   > Set Body SkelMesh(Body SkelMesh:Skeletal Mesh Component Budgeted Object Reference=Get Array Item with EFort Custom Part Type(Array=@n9.CustomMeshParts)) -> (Output_Get:Skeletal Mesh Component Budgeted Object Reference)
   n10 = Event OnEndSkydiving()
   SetClothingInteractorIfNotSet(execute=[@n10.then, @n11.then])
   > SetGravity(self=Get ChaosClothingInteractor(), GravityScale=1, bIsGravityOverridden=false, GravityOverride="0,0,-980")
   n11 = Event OnEndFalling()
   n12 = Event OnBeginSkydivingFromBus() {is=title K2Node_Event}
  ]}
 SetClothingInteractorIfNotSet {nodes [
   SetClothingInteractorIfNotSet() {is=title K2Node_FunctionEntry}
   n1 = IsValid(Object=Get ChaosClothingInteractor())
   > Branch(Condition=AND Boolean!(A=Get bClothEnabled(), B=NOT Boolean(A=@n1)) -> (ReturnValue))
   n2 = MakeLiteralString(Value=F_MED_HomeRange_FaceAcc_Clothing_0)
   n3 = GetOwningComponent()
   > n4 = Cast To ChaosClothingInteractor!(
    execute
    Object=GetClothingInteractor!(
     self:Clothing Simulation Interactor Object Reference=GetClothingSimulationInteractor!(self:Skeletal Mesh Component Object Reference=@n3, ClothingAssetName:Name=None) -> (ReturnValue:Clothing Simulation Interactor Object Reference)
     ClothingAssetName:Name=String To Name(InString=@n2)
    ) -> (ReturnValue:Clothing Interactor Object Reference)
   ) -> (then, CastFailed, AsChaos Clothing Interactor:Chaos Clothing Interactor Object Reference) {is=title K2Node_DynamicCast}
   > Set ChaosClothingInteractor(ChaosClothingInteractor:Chaos Clothing Interactor Object Reference=@n4) -> (Output_Get:Chaos Clothing Interactor Object Reference)
  ]}
]
variables [
 Body SkelMesh {type=SkeletalMeshComponentBudgeted}
 ChaosClothingInteractor {type=ChaosClothingInteractor, default=null}
 FrontEnd Low Setting / IOS devive {type=boolean}
 Gravity Override {type=Vector, default=<0 0 -980>}
 Hat Sim Space Settings {
  type=AnimPhysSimSpaceSettings
  default {maxAngularAcceleration=30, maxAngularVelocity=300, simSpaceAngularAlpha=1, externalAngularVelocity=<0 0 0>}
 }
 Hat Sim Space Settings Default {
  type=AnimPhysSimSpaceSettings
  default {maxAngularAcceleration=30, maxAngularVelocity=300, simSpaceAngularAlpha=1, externalAngularVelocity=<0 0 0>}
 }
 Hat Sim Space Settings Frontend {
  type=AnimPhysSimSpaceSettings
  default {maxAngularAcceleration=30, maxAngularVelocity=300, simSpaceAngularAlpha=0.5, externalAngularVelocity=<0 0 0>}
 }
 isOnGround {type=boolean}
]
