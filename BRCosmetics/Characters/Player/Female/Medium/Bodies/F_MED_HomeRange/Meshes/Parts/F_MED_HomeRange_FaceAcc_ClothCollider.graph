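# Physics asset of collision shapes for the HomeRange female medium character's face accessory, named as a cloth collider.
# Capsules cover the head, neck_01, both clavicles, both dynamic collar bones and the right upper arm, while spine_05 combines a sphere and a capsule.
# The head uses three capsules, and the left upper arm has no body.
# The only constraint, clavicle_l to spine_05, leaves swing and twist free.
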PhysicsAsset /BRCosmetics/Characters/Player/Female/Medium/Bodies/F_MED_HomeRange/Meshes/Parts/F_MED_HomeRange_FaceAcc_ClothCollider
bodies {
 clavicle_l [{
   shapeType=Capsule
   length=9.4831
   radius=6.5948
   shapeName=clavicle_l_capsule
   center=<4.4668 -1.0644 -0.8741>
   rotation=<rot -0.0003 -90 -81.1965>
  }]
 clavicle_r [{
   shapeType=Capsule
   length=9.4831
   radius=6.5948
   shapeName=clavicle_r_capsule
   center=<-4.4668 1.0644 0.8741>
   rotation=<rot -0.0003 -90 98.8035>
  }]
 dyn_collar_l [{
   shapeType=Capsule
   length=8.3244
   radius=3.894
   shapeName=dyn_collar_l_capsule
   center=<0.542 -0.3394 -0.0625>
   rotation=<rot 75.6978 -81.753 0.5779>
  }]
 dyn_collar_r [{
   shapeType=Capsule
   length=8.3244
   radius=3.894
   shapeName=dyn_collar_r_capsule
   center=<-0.542 0.3394 0.0625>
   rotation=<rot 75.6978 -81.753 -179.4221>
  }]
 head [
  {shapeType=Capsule, length=6.6532, radius=6.1169, center=<0.8317 2.2537 0>, rotation=<rot 2.698 90 -90>}
  {shapeType=Capsule, length=19.1888, radius=10.8112, shapeName=head_capsule, center=<19.2805 3.5047 0>}
  {shapeType=Capsule, length=15, radius=15, shapeName=head_capsule, center=<6.8202 -18.4696 0>}
 ]
 neck_01 [{shapeType=Capsule, length=5.8761, radius=5.8057, shapeName=neck_01_capsule, rotation=<rot 0 -90 90>}]
 spine_05 [
  {shapeType=Sphere, radius=11.6158, shapeName=spine_05_sphere, center=<6.4511 -5.1734 0>}
  {shapeType=Capsule, length=13.2432, radius=13.7568, shapeName=spine_05_capsule, center=<-6.0931 -1.7685 0>, rotation=<rot 0 -90 0>}
 ]
 upperarm_r [{
   shapeType=Capsule
   length=15.1497
   radius=5.2772
   shapeName=upperarm_r_capsule
   center=<-9.8265 0.8884 0.6036>
   rotation=<rot -87.2486 -170.3062 165.1669>
  }]
}
constraints [{
  bone1Name=clavicle_l
  bone2Name=spine_05
  swing1LimitDegrees=45
  swing1Motion=Free
  swing2LimitDegrees=45
  swing2Motion=Free
  twistLimitDegrees=45
  twistMotion=Free
 }]
